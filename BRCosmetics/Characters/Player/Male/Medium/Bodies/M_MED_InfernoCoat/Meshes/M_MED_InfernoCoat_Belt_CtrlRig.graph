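# Control Rig that deforms the Inferno Coat belt on the medium male player body.
# Hermite spline IK drives front and back belt bone chains, with control nulls constrained between pelvis and belt-end controls, one following dyn_belt.
# Spherical pose readers supply a smoothed alpha, switched by crouching between left-thigh and belt readings, to blend control poses.
# A spine-twist alpha blends the start control; initialization builds and caches all poses.
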ControlRigBlueprint /BRCosmetics/Characters/Player/Male/Medium/Bodies/M_MED_InfernoCoat/Meshes/M_MED_InfernoCoat_Belt_CtrlRig
graphs [
 ContainedGraph {nodes [
   n1 = RigVMFunctionEntryNode()
   SequenceAggregate = &RigVMFunction_Sequence(ExecuteContext=@n1.ExecuteContext)
   SequenceAggregate_1 = &RigVMFunction_Sequence(ExecuteContext=@SequenceAggregate.B)
   SequenceAggregate_2 = &RigVMFunction_Sequence(ExecuteContext=@SequenceAggregate_1.B)
   SequenceAggregate_3 = &RigVMFunction_Sequence(ExecuteContext=@SequenceAggregate_2.B)
   RigVMFunctionReturnNode(A=@SequenceAggregate.A, B=@SequenceAggregate_1.A, C=@SequenceAggregate_2.A, D=@SequenceAggregate_3.A, E=@SequenceAggregate_3.B)
  ]}
 ContainedGraph {nodes [
   n1 = RigVMFunctionEntryNode()
   SequenceAggregate = &RigVMFunction_Sequence(ExecuteContext=@n1.ExecuteContext)
   SequenceAggregate_1 = &RigVMFunction_Sequence(ExecuteContext=@SequenceAggregate.B)
   SequenceAggregate_2 = &RigVMFunction_Sequence(ExecuteContext=@SequenceAggregate_1.B)
   RigVMFunctionReturnNode(A=@SequenceAggregate.A, B=@SequenceAggregate_1.A, C=@SequenceAggregate_2.A, D=@SequenceAggregate_2.B)
  ]}
 RigVMFunctionLibrary {links [], nodes []}
 RigVMModel {nodes [
   n1 = &RigUnit_BeginExecution()
   SequenceAggregate = RigVMAggregateNode(ExecuteContext=@n1.ExecutePin)
   n2 = &RigUnit_SphericalPoseReader(
    ActiveRegionScaleFactors=(PositiveWidth=3.5,NegativeWidth=0.75,PositiveHeight=3,NegativeHeight=3)
    ActiveRegionSize=0.1
    Debug=(bDrawDebug=false,DebugScale=25,DebugSegments=20,DebugThickness=0.25)
    DriverAxis=(X=-1,Y=0,Z=0)
    DriverItem=(Type=Bone,Name="thigh_l")
    FalloffRegionScaleFactors=(PositiveWidth=1,NegativeWidth=1,PositiveHeight=1,NegativeHeight=1)
    FalloffSize=0.45
    OptionalParentItem=(Type=Bone)
    RotationOffset=(X=90,Y=0,Z=0)
    ExecutePin=@SequenceAggregate.A
   )
   > RigVMVariableNode(Value=@n2.OutputParam) {variable=spr_thigh_l_up_alpha}
   > n3 = &RigUnit_SphericalPoseReader(
    ActiveRegionScaleFactors=(PositiveWidth=3.5,NegativeWidth=0.5,PositiveHeight=3,NegativeHeight=3)
    ActiveRegionSize=0.1
    Debug=(bDrawDebug=false,DebugScale=25,DebugSegments=20,DebugThickness=0.25)
    DriverAxis=(X=1,Y=0,Z=0)
    DriverItem=(Type=Bone,Name="thigh_r")
    FalloffRegionScaleFactors=(PositiveWidth=1,NegativeWidth=1,PositiveHeight=1,NegativeHeight=1)
    FalloffSize=0.45
    FlipHeightScaling=false
    FlipWidthScaling=true
    OptionalParentItem=(Type=Bone)
    RotationOffset=(X=-90,Y=0,Z=0)
   )
   > RigVMVariableNode(Value=@n3.OutputParam) {variable=spr_thigh_r_up_alpha}
   > n4 = &RigUnit_SphericalPoseReader(
    ActiveRegionScaleFactors=(PositiveWidth=3.6,NegativeWidth=1.4,PositiveHeight=3.6,NegativeHeight=3.6)
    ActiveRegionSize=0.025
    Debug=(bDrawDebug=false,DebugScale=25,DebugSegments=20,DebugThickness=0.25)
    DriverAxis=(X=1,Y=0,Z=0)
    DriverItem=(Type=Bone,Name="dyn_belt_mstr")
    FalloffRegionScaleFactors=(PositiveWidth=1,NegativeWidth=1,PositiveHeight=1,NegativeHeight=1)
    FalloffSize=0.09
    OptionalParentItem=(Type=Bone)
    RotationOffset=(X=90,Y=0,Z=-107)
   )
   > RigVMVariableNode(Value=@n4.OutputParam) {variable=belt_distance_alpha}
   RerouteNode = RigVMDispatchNode(Values="((Type=Bone,Name=\"dfrm_belt_front_01\"),((),Type=Bone,Name=\"dfrm_belt_front_02\"),((),Type=Bone,Name=\"dfrm_belt_front_03\"),((),Type=Bone,Name=\"dfrm_belt_front_04\"),((),Type=Bone,Name=\"dfrm_belt_front_05\"),((),Type=Bone,Name=\"dfrm_belt_front_06\"),((),Type=Bon ...(truncated)") {template="DISPATCH_RigVMDispatch_ArrayMake(in Values,out Array)"}
   SplineIK = RigVMFunctionReferenceNode(
    Primary Axis=(X=1,Y=0,Z=0)
    Samples Per Segment=32
    Secondary Spline Direction=(X=0,Y=3,Z=0)
    Spline Mode=Hermite
    Stretch=true
    Up Axis=(X=0,Y=1,Z=0)
    debug=false
    Bones=@RerouteNode.Array
    ExecuteContext=@SequenceAggregate.B
    Controls=RigVMVariableNode() {variable=front_spline_ctrls}.Value
   )
   RerouteNode_3 = RigVMDispatchNode(Values="((Type=Bone,Name=\"dfrm_belt_back_01\"),((),Type=Bone,Name=\"dfrm_belt_back_02\"),((),Type=Bone,Name=\"dfrm_belt_back_03\"),((),Type=Bone,Name=\"dfrm_belt_back_04\"),((),Type=Bone,Name=\"dfrm_belt_back_05\"),((),Type=Bone,Name=\"dfrm_belt_back_06\"),((),Type=Bone,Name ...(truncated)") {template="DISPATCH_RigVMDispatch_ArrayMake(in Values,out Array)"}
   > SplineIK_1 = RigVMFunctionReferenceNode(
    Primary Axis=(X=1,Y=0,Z=0)
    Samples Per Segment=32
    Secondary Spline Direction=(X=0,Y=3,Z=0)
    Spline Mode=Hermite
    Stretch=true
    Up Axis=(X=0,Y=1,Z=0)
    debug=false
    Bones=@RerouteNode_3.Array
    Controls=RigVMVariableNode() {variable=back_spline_ctrls}.Value
   )
   SequenceAggregate_1 = &RigVMFunction_Sequence(ExecuteContext=@SequenceAggregate.C)
   &RigUnit_ParentConstraint(
    AdvancedSettings=(RotationOrderForFilter=XZY)
    Child=(Type=Null,Name="belt_end_ctrl_null")
    Parents=((Item=(Type=Bone,Name="dyn_belt"),Weight=1))
    Weight=1
    ExecutePin=@SequenceAggregate_1.A
   )
   > &RigUnit_ParentConstraint(
    AdvancedSettings=(RotationOrderForFilter=XZY)
    Child=(Type=Null,Name="dfrm_belt_front_03_ctrl_null")
    Parents=((Item=(Type=Control,Name="belt_start_ctrl"),Weight=0.8),(Item=(Type=Bone,Name="pelvis"),Weight=0.2))
    Weight=1
   )
   > &RigUnit_ParentConstraint(
    AdvancedSettings=(RotationOrderForFilter=XZY)
    Child=(Type=Null,Name="dfrm_belt_front_05_ctrl_null")
    Parents=((Item=(Type=Control,Name="belt_start_ctrl"),Weight=0.2),(Item=(Type=Bone,Name="pelvis"),Weight=0.8))
    Weight=1
   )
   > &RigUnit_ParentConstraint(
    AdvancedSettings=(RotationOrderForFilter=XZY)
    Child=(Type=Null,Name="dfrm_belt_front_06_ctrl_null")
    Parents=((Item=(Type=Control,Name="belt_end_ctrl"),Weight=0.15),(Item=(Type=Bone,Name="pelvis"),Weight=0.85))
    Weight=1
   )
   > &RigUnit_ParentConstraint(
    AdvancedSettings=(RotationOrderForFilter=XZY)
    Child=(Type=Null,Name="dfrm_belt_front_08_ctrl_null")
    Parents=((Item=(Type=Control,Name="belt_end_ctrl"),Weight=0.5),(Item=(Type=Bone,Name="pelvis"),Weight=0.5))
    Weight=1
   )
   &RigUnit_ParentConstraint(
    AdvancedSettings=(RotationOrderForFilter=XZY)
    Child=(Type=Null,Name="dfrm_belt_back_03_ctrl_null")
    Parents=((Item=(Type=Control,Name="belt_start_ctrl"),Weight=0.8),(Item=(Type=Bone,Name="pelvis"),Weight=0.2))
    Weight=1
    ExecutePin=@SequenceAggregate_1.B
   )
   > &RigUnit_ParentConstraint(
    AdvancedSettings=(RotationOrderForFilter=XZY)
    Child=(Type=Null,Name="dfrm_belt_back_05_ctrl_null")
    Parents=((Item=(Type=Control,Name="belt_start_ctrl"),Weight=0.2),(Item=(Type=Bone,Name="pelvis"),Weight=0.8))
    Weight=1
   )
   > &RigUnit_ParentConstraint(
    AdvancedSettings=(RotationOrderForFilter=XZY)
    Child=(Type=Null,Name="dfrm_belt_back_08_ctrl_null")
    Parents=((Item=(Type=Control,Name="belt_end_ctrl"),Weight=0.6),(Item=(Type=Bone,Name="pelvis"),Weight=0.4))
    Weight=1
   )
   SequenceAggregate_2 = &RigVMFunction_Sequence(ExecuteContext=@SequenceAggregate.D)
   Cast = &RigVMFunction_MathBoolToFloat(Value=RigVMVariableNode() {variable=is_crouching}.Value)
   Multiply = &RigVMFunction_MathDoubleMul(A=@Cast.Result, B=RigVMVariableNode() {variable=spr_thigh_l_up_alpha}.Value)
   Subtract = &RigVMFunction_MathDoubleSub(A=1, B=@Cast.Result)
   Maximum = &RigVMFunction_MathDoubleMax(
    A=@Multiply.Result
    B=&RigVMFunction_MathDoubleMul(A=@Subtract.Result, B=RigVMVariableNode() {variable=belt_distance_alpha}.Value).Result
   )
   n5 = &RigVMFunction_AlphaInterp(InterpSpeedDecreasing=20, InterpSpeedIncreasing=10, Scale=1, bClampResult=False, Value=@Maximum.Result)
   n6 = RigVMVariableNode() {variable=belt_spline_ctrls}
   /Script/ControlRig.RigUnit_HierarchySetPoseItemArray(
    ElementType=All
    ExecutePin=@SequenceAggregate_2.A
    Weight=&RigVMFunction_MathDoubleSub(A=1, B=@n5.Result).Result
    Pose=RigVMVariableNode() {variable=belt_IKSpline_bindpose}.Value
    ItemsToSet=@n6.Value
   ) {is=struct RigVMUnitNode}
   > &RigUnit_HierarchySetPoseItemArray(ElementType=All, Weight=@n5.Result, Pose=RigVMVariableNode() {variable=front_IKSpline_thigh_l_up_pose}.Value, ItemsToSet=@n6.Value)
   n7 = RigVMVariableNode() {variable=spine_twist_pos_delta_ref_alpha}
   &RigUnit_HierarchySetPoseItemArray(
    ElementType=All
    ItemsToSet=((Type=Control,Name="belt_start_ctrl"))
    Weight=&RigVMFunction_MathDoubleSub(A=1, B=@n7.Value).Result
    Pose=RigVMVariableNode() {variable=front_IKSpline_ctrl_bindpose}.Value
    ExecutePin=@SequenceAggregate_2.B
   )
   > &RigUnit_HierarchySetPoseItemArray(
    ElementType=All
    ItemsToSet=((Type=Control,Name="belt_start_ctrl"))
    Weight=@n7.Value
    Pose=RigVMVariableNode() {variable=back_belt_body_twist_left_pose}.Value
   )
  ]}
 RigVMModel initialize_pose_cache_variables {nodes [
   n1 = &RigUnit_PrepareForExecution()
   SequenceAggregate = RigVMAggregateNode(ExecuteContext=@n1.ExecutePin)
   &RigUnit_ModifyTransforms(
    ItemToModify=((Item=(Type=Control,Name="belt_start_ctrl"),Transform=(Rotation=(X=0,Y=0,Z=0,W=1),Translation=(X=0,Y=-0.5,Z=-0.5),Scale3D=(X=1,Y=1,Z=1))))
    Weight=1
    ExecutePin=@SequenceAggregate.A
   )
   > RigVMVariableNode(Value=/Script/ControlRig.RigUnit_HierarchyGetPoseItemArray(ElementType=All, ItemsToGet=RigVMVariableNode() {variable=belt_spline_ctrls}.Value) {is=struct RigVMUnitNode}.Pose) {variable=belt_IKSpline_bindpose}
   SequenceAggregate_1 = &RigVMFunction_Sequence(ExecuteContext=@SequenceAggregate.C)
   &RigUnit_ModifyTransforms(
    ItemToModify=((Item=(Type=Control,Name="belt_end_ctrl"),Transform=(Rotation=(X=0,Y=0,Z=0,W=1),Translation=(X=0,Y=4,Z=4),Scale3D=(X=1,Y=1,Z=1))))
    Weight=1
    ItemToModify.0.Transform.Rotation=&RigVMFunction_MathQuaternionFromEuler(Euler=(X=0,Y=-3,Z=-5)).Result
    ExecutePin=@SequenceAggregate_1.A
   )
   > &RigUnit_ModifyTransforms(
    ItemToModify=((Item=(Type=Control,Name="dfrm_belt_front_05_ctrl"),Transform=(Rotation=(X=0,Y=0,Z=0,W=1),Translation=(X=-2.5,Y=0,Z=-1.5),Scale3D=(X=1,Y=1,Z=1))))
    Weight=1
   )
   > &RigUnit_ModifyTransforms(
    ItemToModify=((Item=(Type=Control,Name="dfrm_belt_front_06_ctrl"),Transform=(Rotation=(X=0,Y=0,Z=0,W=1),Translation=(X=-1,Y=-1.5,Z=-4.5),Scale3D=(X=1,Y=1,Z=1))))
    Weight=1
   )
   > &RigUnit_ModifyTransforms(
    ItemToModify=((Item=(Type=Control,Name="dfrm_belt_front_08_ctrl"),Transform=(Rotation=(X=0,Y=0,Z=0,W=1),Translation=(X=0,Y=-1,Z=-6),Scale3D=(X=1,Y=1,Z=1))))
    Weight=1
   )
   > &RigUnit_ModifyTransforms(
    ItemToModify=((Item=(Type=Control,Name="dfrm_belt_front_10_ctrl"),Transform=(Rotation=(X=0,Y=0,Z=0,W=1),Translation=(X=0,Y=0,Z=-1),Scale3D=(X=1,Y=1,Z=1))))
    Weight=1
   )
   &RigUnit_ModifyTransforms(
    ItemToModify=((Item=(Type=Control,Name="dfrm_belt_back_11_ctrl"),Transform=(Rotation=(X=0,Y=0,Z=0,W=1),Translation=(X=-1.25,Y=-0.5,Z=-1.5),Scale3D=(X=1,Y=1,Z=1))))
    Weight=1
    ExecutePin=@SequenceAggregate_1.B
   )
   > &RigUnit_ModifyTransforms(
    ItemToModify=((Item=(Type=Control,Name="dfrm_belt_back_08_ctrl"),Transform=(Rotation=(X=0,Y=0,Z=0,W=1),Translation=(X=0,Y=0,Z=1.25),Scale3D=(X=1,Y=1,Z=1))))
    Weight=1
   )
   > &RigUnit_ModifyTransforms(
    ItemToModify=((Item=(Type=Control,Name="dfrm_belt_back_05_ctrl"),Transform=(Rotation=(X=0,Y=0,Z=0,W=1),Translation=(X=0,Y=0,Z=2),Scale3D=(X=1,Y=1,Z=1))))
    Weight=1
   )
   > RigVMVariableNode(Value=&RigUnit_HierarchyGetPoseItemArray(ElementType=All, ItemsToGet=RigVMVariableNode() {variable=belt_spline_ctrls}.Value).Pose) {variable=front_IKSpline_thigh_l_up_pose}
   SequenceAggregate_2 = &RigVMFunction_Sequence(ExecuteContext=@SequenceAggregate.D)
   &RigUnit_ModifyTransforms(
    ItemToModify=((Item=(Type=Control,Name="belt_start_ctrl"),Transform=(Rotation=(X=0,Y=0,Z=0,W=1),Translation=(X=0,Y=-1.5,Z=-0.5),Scale3D=(X=1,Y=1,Z=1))))
    Weight=1
    ItemToModify.0.Transform.Rotation=&RigVMFunction_MathQuaternionFromEuler(Euler=(X=8,Y=0,Z=0)).Result
    ExecutePin=@SequenceAggregate_2.A
   )
   > RigVMVariableNode(Value=&RigUnit_HierarchyGetPoseItemArray(ElementType=All, ItemsToGet=RigVMVariableNode() {variable=front_spline_ctrls}.Value).Pose) {variable=back_belt_body_twist_left_pose}
  ]}
]
hierarchy [
 root {loc=<0 0 0>, rot=<0 0 0 1>}
 pelvis {parent=root, loc=<0 -0.7071 93.1191>, rot=<0.0127 0.707 -0.0127 -0.707>}
 spine_01 {parent=pelvis, loc=<6.7494 -0.6388 0>, rot=<0 0 0.034 -0.9994>}
 spine_02 {parent=spine_01, loc=<7.2518 -0.035 0>, rot=<0 0 0.0238 -0.9997>}
 spine_03 {parent=spine_02, loc=<6.0104 0 0>, rot=<0 0 0.0149 -0.9999>}
 spine_04 {parent=spine_03, loc=<8.1024 0 0>, rot=<0 0 -0.0132 -0.9999>}
 spine_05 {parent=spine_04, loc=<10.0146 0 0>, rot=<0 0 -0.0015 -1>}
 attach_cape {parent=spine_05, loc=<9.017 -11.714 0>, rot=<0.0616 -0.7044 -0.0616 -0.7044>}
 neck_01 {parent=spine_05, loc=<11.3138 0.1529 0>, rot=<0 0 -0.2003 -0.9797>}
 neck_02 {parent=neck_01, loc=<4.5627 0 0>, rot=<0 0 0.0504 -0.9987>}
 head {parent=neck_02, loc=<5.3914 0 0>, rot=<0 0 0.115 -0.9934>}
 attach_backpack {parent=spine_05, loc=<-0.123 -13.473 0>, rot=<-0.0093 -0.707 0.0093 -0.707>}
 clavicle_l {parent=spine_05, loc=<4.3575 -0.1466 -2.3223>, rot=<0.0465 -0.7289 -0.009 -0.6829>}
 upperarm_l {parent=clavicle_l, loc=<15.5111 0 0>, rot=<-0.0678 -0.3171 -0.2704 -0.9065>}
 lowerarm_l {parent=upperarm_l, loc=<26.6514 0 0>, rot=<0.0391 -0.0099 -0.2453 -0.9686>}
 hand_l {parent=lowerarm_l, loc=<22.3491 0 0>, rot=<-0.7081 -0.0507 0.0975 -0.6975>}
 thumb_01_l {parent=hand_l, loc=<1.3118 -0.3191 -2.459>, rot=<0.6546 -0.4042 0.0664 -0.6354>}
 thumb_02_l {parent=thumb_01_l, loc=<5.9336 -0.0448 0.0401>, rot=<-0.0037 0.0213 0.298 -0.9543>}
 thumb_03_l {parent=thumb_02_l, loc=<2.2223 0.3682 -0.0907>, rot=<0.0118 -0.013 0.0066 -0.9998>}
 index_metacarpal_l {parent=hand_l, loc=<1.6535 0.0183 -2.96>, rot=<0.0158 -0.0124 -0.0173 -0.9996>}
 index_01_l {parent=index_metacarpal_l, loc=<8.3008 0.0357 -0.0524>, rot=<-0.0044 0.0485 0.0924 -0.9945>}
 index_02_l {parent=index_01_l, loc=<3.3587 -0.3004 0.0452>, rot=<-0.002 0.0027 0.3225 -0.9466>}
 index_03_l {parent=index_02_l, loc=<2.5578 -0.123 -0.0252>, rot=<-0.002 0.0067 0.2558 -0.9667>}
 middle_metacarpal_l {parent=hand_l, loc=<2.1856 -0.1725 -0.5984>, rot=<0.006 0.0168 -0.0329 -0.9993>}
 middle_01_l {parent=middle_metacarpal_l, loc=<7.3275 -0.2447 0.0129>, rot=<-0.0208 0.0042 0.0858 -0.9961>}
 middle_02_l {parent=middle_01_l, loc=<4.0123 0.1401 0.1175>, rot=<-0.0218 0.0258 0.3512 -0.9357>}
 middle_03_l {parent=middle_02_l, loc=<2.6077 0.0534 0.1373>, rot=<0.0082 0.0188 0.1708 -0.9851>}
 ring_metacarpal_l {parent=hand_l, loc=<2.2074 -0.3625 1.3638>, rot=<-0.0284 0.0722 -0.0458 -0.9959>}
 ring_01_l {parent=ring_metacarpal_l, loc=<7.0295 -0.6197 -0.1446>, rot=<-0.0489 -0.0252 0.1892 -0.9804>}
 ring_02_l {parent=ring_01_l, loc=<3.6174 0.4647 -0.1027>, rot=<0 0 0.2806 -0.9598>}
 ring_03_l {parent=ring_02_l, loc=<2.4152 0.114 0.0199>, rot=<0 0 0.1189 -0.9929>}
 pinky_metacarpal_l {parent=hand_l, loc=<1.7616 -0.9753 3.3814>, rot=<-0.0397 0.0814 -0.069 -0.9935>}
 pinky_01_l {parent=pinky_metacarpal_l, loc=<7.1709 -0.4831 -0.1094>, rot=<-0.0654 -0.0035 0.2242 -0.9723>}
 pinky_02_l {parent=pinky_01_l, loc=<2.8274 0.0893 -0.1043>, rot=<0.0037 0.0011 0.2854 -0.9584>}
 pinky_03_l {parent=pinky_02_l, loc=<2.2046 0.0307 -0.0078>, rot=<-0.0002 0.0098 0.1445 -0.9895>}
 deform_wrist_dn_base_l {parent=hand_l, loc=<0 -2.2342 0>, rot=<0 0 0 -1>}
 deform_wrist_dn_l {parent=deform_wrist_dn_base_l, loc=<0 0 0>, rot=<0 0 0 1>}
 deform_wrist_up_base_l {parent=hand_l, loc=<0 2.2682 0>, rot=<0 0 0 -1>}
 deform_wrist_up_l {parent=deform_wrist_up_base_l, loc=<0 0 0>, rot=<0 0 0 1>}
 lowerarm_twist_02_l {parent=lowerarm_l, loc=<9.6199 0 0>, rot=<0 0 0 -1>}
 lowerarm_twist_01_l {parent=lowerarm_l, loc=<17.8549 0 0>, rot=<0 0 0 -1>}
 deform_elbow_in_base_l {parent=lowerarm_l, loc=<0.2112 5.259 0.8274>, rot=<0 0 0 -1>}
 deform_elbow_in_l {parent=deform_elbow_in_base_l, loc=<0 0 0>, rot=<0 0 0 1>}
 deform_elbow_out_base_l {parent=lowerarm_l, loc=<0 -4 0>, rot=<0 0 0 -1>}
 deform_elbow_out_l {parent=deform_elbow_out_base_l, loc=<0 0 0>, rot=<0 0 0 1>}
 upperarm_twist_01_l {parent=upperarm_l, loc=<0.0041 0 0>, rot=<0 0 0 1>}
 deform_upperarm_bk_base_l {parent=upperarm_twist_01_l, loc=<-1.7241 -6.8488 -1.9453>, rot=<0 0 0.1951 -0.9808>}
 deform_upperarm_bk_l {parent=deform_upperarm_bk_base_l, loc=<0 0 0>, rot=<0 0 0 1>}
 deform_upperarm_fr_base_l {parent=upperarm_twist_01_l, loc=<2.2962 6.4299 -2.765>, rot=<0 0 0.1951 -0.9808>}
 deform_upperarm_fr_l {parent=deform_upperarm_fr_base_l, loc=<0 0 0>, rot=<0 0 0 1>}
 deform_upperarm_up_base_l {parent=upperarm_twist_01_l, loc=<-3.2668 -0.3845 3.3311>, rot=<0 0 0.112 -0.9937>}
 deform_upperarm_up_l {parent=deform_upperarm_up_base_l, loc=<0 0 0>, rot=<0 0 0 1>}
 upperarm_twist_02_l {parent=upperarm_l, loc=<13.7941 0 0>, rot=<0 0 0 1>}
 dfrm_clavicle_l {parent=clavicle_l, loc=<0 0 0>, rot=<0 0 0 1>}
 dfrm_furr_clavicle_l {parent=clavicle_l, loc=<0 0 0>, rot=<0 0 0 1>}
 dyn_mstr_fur_A_l {parent=dfrm_furr_clavicle_l, loc=<6.0622 3.8872 11.6383>, rot=<0.6444 -0.3346 0.5455 -0.4186>}
 dyn_cons_fur_A_01_l {parent=dyn_mstr_fur_A_l, loc=<10.3377 -1.503 -3.3368>, rot=<0.142 0.0855 0.3453 -0.9237>}
 dyn_mstr_fur_A_01_l {parent=dyn_cons_fur_A_01_l, loc=<0 0 0>, rot=<0 0 0 1>}
 dyn_fur_A_01_l {parent=dyn_mstr_fur_A_01_l, loc=<0 0 0>, rot=<0 0 0 1>}
 dyn_cons_fur_A_02_l {parent=dyn_mstr_fur_A_l, loc=<13.2654 2.2742 -4.2681>, rot=<0.0999 -0.1422 0.1701 -0.97>}
 dyn_mstr_fur_A_02_l {parent=dyn_cons_fur_A_02_l, loc=<0 0 0>, rot=<0 0 0 1>}
 dyn_fur_A_02_l {parent=dyn_mstr_fur_A_02_l, loc=<0 0 0>, rot=<0 0 0 1>}
 dyn_cons_fur_A_03_l {parent=dyn_mstr_fur_A_l, loc=<6.4254 -2.0872 -0.0913>, rot=<0.1267 -0.1407 0.3141 -0.9303>}
 dyn_mstr_fur_A_03_l {parent=dyn_cons_fur_A_03_l, loc=<0 0 0>, rot=<0 0 0 1>}
 dyn_fur_A_03_l {parent=dyn_mstr_fur_A_03_l, loc=<0 0 0>, rot=<0 0 0 1>}
 dyn_mstr_fur_B_l {parent=dfrm_furr_clavicle_l, loc=<9.3988 7.1068 9.7526>, rot=<0.7849 -0.3806 0.4324 -0.2285>}
 dyn_cons_fur_B_01_l {parent=dyn_mstr_fur_B_l, loc=<12.4893 14.4193 1.3248>, rot=<-0.1791 -0.1394 -0.5369 -0.8125>}
 dyn_mstr_fur_B_01_l {parent=dyn_cons_fur_B_01_l, loc=<0 0 0>, rot=<0 0 0 1>}
 dyn_fur_B_01_l {parent=dyn_mstr_fur_B_01_l, loc=<0 0 0>, rot=<0 0 0 1>}
 dyn_cons_fur_B_02_l {parent=dyn_mstr_fur_B_l, loc=<14.1242 10.5332 1.3895>, rot=<-0.2145 -0.0843 -0.2153 -0.949>}
 dyn_mstr_fur_B_02_l {parent=dyn_cons_fur_B_02_l, loc=<0 0 0>, rot=<0 0 0 1>}
 dyn_fur_B_02_l {parent=dyn_mstr_fur_B_02_l, loc=<0 0 0>, rot=<0 0 0 1>}
 dyn_cons_fur_B_03_l {parent=dyn_mstr_fur_B_l, loc=<14.3629 7.1402 -2.1132>, rot=<-0.0992 -0.3049 -0.0933 -0.9426>}
 dyn_mstr_fur_B_03_l {parent=dyn_cons_fur_B_03_l, loc=<0 0 0>, rot=<0 0 0 1>}
 dyn_fur_B_03_l {parent=dyn_mstr_fur_B_03_l, loc=<0 0 0>, rot=<0 0 0 1>}
 dyn_cons_fur_B_04_l {parent=dyn_mstr_fur_B_l, loc=<9.9593 1.3837 -0.3497>, rot=<0.0303 -0.2037 0.2161 -0.9544>}
 dyn_mstr_fur_B_04_l {parent=dyn_cons_fur_B_04_l, loc=<0 0 0>, rot=<0 0 0 1>}
 dyn_fur_B_04_l {parent=dyn_mstr_fur_B_04_l, loc=<0 0 0>, rot=<0 0 0 1>}
 dyn_cons_fur_B_05_l {parent=dyn_mstr_fur_B_l, loc=<6.6029 -0.7484 1.8097>, rot=<-0.0735 -0.0373 0.22 -0.972>}
 dyn_mstr_fur_B_05_l {parent=dyn_cons_fur_B_05_l, loc=<0 0 0>, rot=<0 0 0 1>}
 dyn_fur_B_05_l {parent=dyn_mstr_fur_B_05_l, loc=<0 0 0>, rot=<0 0 0 1>}
 dyn_cons_fur_B_06_l {parent=dyn_mstr_fur_B_l, loc=<5.1424 0.459 3.8333>, rot=<0.026 -0.0013 0.1079 -0.9938>}
 dyn_mstr_fur_B_06_l {parent=dyn_cons_fur_B_06_l, loc=<0 0 0>, rot=<0 0 0 1>}
 dyn_fur_B_06_l {parent=dyn_mstr_fur_B_06_l, loc=<0 0 0>, rot=<0 0 0 1>}
 dyn_cons_fur_B_07_l {parent=dyn_mstr_fur_B_l, loc=<0.9494 1.2141 6.1481>, rot=<0.0034 0.0147 0.0494 -0.9987>}
 dyn_mstr_fur_B_07_l {parent=dyn_cons_fur_B_07_l, loc=<0 0 0>, rot=<0 0 0 1>}
 dyn_fur_B_07_l {parent=dyn_mstr_fur_B_07_l, loc=<0 0 0>, rot=<0 0 0 1>}
 dyn_cons_fur_B_08_l {parent=dyn_mstr_fur_B_l, loc=<3.533 -2.1299 -1.8637>, rot=<-0.1089 -0.2272 0.0608 -0.9658>}
 dyn_mstr_fur_B_08_l {parent=dyn_cons_fur_B_08_l, loc=<0 0 0>, rot=<0 0 0 1>}
 dyn_fur_B_08_l {parent=dyn_mstr_fur_B_08_l, loc=<0 0 0>, rot=<0 0 0 1>}
 clavicle_r {parent=spine_05, loc=<4.3575 -0.1466 2.3223>, rot=<0.7289 0.0465 -0.6829 0.009>}
 upperarm_r {parent=clavicle_r, loc=<-15.5111 0 0>, rot=<0.0226 -0.3234 -0.0088 -0.9459>}
 lowerarm_r {parent=upperarm_r, loc=<-26.6514 0 0>, rot=<0.0391 -0.0099 -0.2453 -0.9686>}
 hand_r {parent=lowerarm_r, loc=<-22.3491 0 0>, rot=<-0.7081 -0.0507 0.0975 -0.6975>}
 thumb_01_r {parent=hand_r, loc=<-1.3118 0.3191 2.459>, rot=<0.6546 -0.4042 0.0664 -0.6354>}
 thumb_02_r {parent=thumb_01_r, loc=<-5.9336 0.0448 -0.0401>, rot=<-0.0037 0.0213 0.298 -0.9543>}
 thumb_03_r {parent=thumb_02_r, loc=<-2.2223 -0.3682 0.0907>, rot=<0.0118 -0.013 0.0066 -0.9998>}
 index_metacarpal_r {parent=hand_r, loc=<-1.6535 -0.0183 2.96>, rot=<0.0158 -0.0124 -0.0173 -0.9996>}
 index_01_r {parent=index_metacarpal_r, loc=<-8.3008 -0.0357 0.0524>, rot=<-0.0044 0.0485 0.0924 -0.9945>}
 index_02_r {parent=index_01_r, loc=<-3.3587 0.3004 -0.0452>, rot=<-0.002 0.0027 0.3225 -0.9466>}
 index_03_r {parent=index_02_r, loc=<-2.5578 0.123 0.0252>, rot=<-0.002 0.0067 0.2558 -0.9667>}
 middle_metacarpal_r {parent=hand_r, loc=<-2.1856 0.1725 0.5984>, rot=<0.006 0.0168 -0.0329 -0.9993>}
 middle_01_r {parent=middle_metacarpal_r, loc=<-7.3275 0.2447 -0.0129>, rot=<-0.0208 0.0042 0.0858 -0.9961>}
 middle_02_r {parent=middle_01_r, loc=<-4.0123 -0.1401 -0.1175>, rot=<-0.0218 0.0258 0.3512 -0.9357>}
 middle_03_r {parent=middle_02_r, loc=<-2.6077 -0.0534 -0.1373>, rot=<0.0082 0.0188 0.1708 -0.9851>}
 ring_metacarpal_r {parent=hand_r, loc=<-2.2074 0.3625 -1.3638>, rot=<-0.0284 0.0722 -0.0458 -0.9959>}
 ring_01_r {parent=ring_metacarpal_r, loc=<-7.0295 0.6197 0.1446>, rot=<-0.0489 -0.0252 0.1892 -0.9804>}
 ring_02_r {parent=ring_01_r, loc=<-3.6174 -0.4647 0.1027>, rot=<0 0 0.2806 -0.9598>}
 ring_03_r {parent=ring_02_r, loc=<-2.4152 -0.114 -0.0199>, rot=<0 0 0.1189 -0.9929>}
 pinky_metacarpal_r {parent=hand_r, loc=<-1.7616 0.9753 -3.3814>, rot=<-0.0397 0.0814 -0.069 -0.9935>}
 pinky_01_r {parent=pinky_metacarpal_r, loc=<-7.1709 0.4831 0.1094>, rot=<-0.0654 -0.0035 0.2242 -0.9723>}
 pinky_02_r {parent=pinky_01_r, loc=<-2.8274 -0.0893 0.1043>, rot=<0.0037 0.0011 0.2854 -0.9584>}
 pinky_03_r {parent=pinky_02_r, loc=<-2.2046 -0.0307 0.0078>, rot=<-0.0002 0.0098 0.1445 -0.9895>}
 deform_wrist_dn_base_r {parent=hand_r, loc=<-0.0004 2.2345 -0.0001>, rot=<0 0 0 -1>}
 deform_wrist_dn_r {parent=deform_wrist_dn_base_r, loc=<0 0 0>, rot=<0 0 0 1>}
 deform_wrist_up_base_r {parent=hand_r, loc=<0.0003 -2.2684 0.0001>, rot=<0 0 0 -1>}
 deform_wrist_up_r {parent=deform_wrist_up_base_r, loc=<0 0 0>, rot=<0 0 0 1>}
 lowerarm_twist_02_r {parent=lowerarm_r, loc=<-9.6199 0 0>, rot=<0 0 0 -1>}
 lowerarm_twist_01_r {parent=lowerarm_r, loc=<-17.8549 0 0>, rot=<0 0 0 -1>}
 deform_elbow_in_base_r {parent=lowerarm_r, loc=<-0.2111 -5.2591 -0.8275>, rot=<0 0 0 -1>}
 deform_elbow_in_r {parent=deform_elbow_in_base_r, loc=<-0.0001 0 0.0002>, rot=<0 0 0 1>}
 deform_elbow_out_base_r {parent=lowerarm_r, loc=<0 4 -0.0001>, rot=<0 0 0 -1>}
 deform_elbow_out_r {parent=deform_elbow_out_base_r, loc=<0 0 0>, rot=<0 0 0 1>}
 upperarm_twist_01_r {parent=upperarm_r, loc=<-0.0041 0 0>, rot=<0 0 0 -1>}
 deform_upperarm_bk_base_r {parent=upperarm_twist_01_r, loc=<-1.0814 7.2192 1.9455>, rot=<0 0 0 1>}
 deform_upperarm_bk_r {parent=deform_upperarm_bk_base_r, loc=<0 0 0>, rot=<0 0 0 1>}
 deform_upperarm_fr_base_r {parent=upperarm_twist_01_r, loc=<0.3689 -7.0506 2.7652>, rot=<0 0 0 1>}
 deform_upperarm_fr_r {parent=deform_upperarm_fr_base_r, loc=<0 0 0>, rot=<0 0 0 1>}
 deform_upperarm_up_base_r {parent=upperarm_twist_01_r, loc=<3.2214 1.1391 -3.3308>, rot=<0 0 0 1>}
 deform_upperarm_up_r {parent=deform_upperarm_up_base_r, loc=<0 0 0>, rot=<0 0 0 -1>}
 upperarm_twist_02_r {parent=upperarm_r, loc=<-13.7941 0 0>, rot=<0 0 0 -1>}
 dfrm_clavicle_r {parent=clavicle_r, loc=<0 0 0>, rot=<0 0 0 1>}
 dfrm_furr_clavicle_r {parent=clavicle_r, loc=<0 0 0>, rot=<0 0 0 1>}
 dyn_cons_fur_A_r {parent=dfrm_furr_clavicle_r, loc=<-5.3026 -5.2112 -10.6545>, rot=<-0.1135 -0.6989 -0.3046 -0.6371>}
 dyn_cons_fur_A_01_r {parent=dyn_cons_fur_A_r, loc=<11.8415 4.9888 3.2461>, rot=<0.1579 0.1077 0.0777 -0.9785>}
 dyn_mstr_fur_A_01_r {parent=dyn_cons_fur_A_01_r, loc=<0 0 0>, rot=<0 0 0 1>}
 dyn_fur_A_01_r {parent=dyn_mstr_fur_A_01_r, loc=<0 0 0>, rot=<0 0 0 1>}
 dyn_cons_fur_A_02_r {parent=dyn_cons_fur_A_r, loc=<11.2494 7.0317 4.1154>, rot=<0.0615 0.2383 0.0821 -0.9658>}
 dyn_mstr_fur_A_02_r {parent=dyn_cons_fur_A_02_r, loc=<0 0 0>, rot=<0 0 0 1>}
 dyn_fur_A_02_r {parent=dyn_mstr_fur_A_02_r, loc=<0 0 0>, rot=<0 0 0 1>}
 dyn_cons_fur_A_03_r {parent=dyn_cons_fur_A_r, loc=<2.1341 -1.6438 -3.302>, rot=<0.026 0.0371 0.0292 -0.9985>}
 dyn_mstr_fur_A_03_r {parent=dyn_cons_fur_A_03_r, loc=<0 0 0>, rot=<0 0 0 1>}
 dyn_fur_A_03_r {parent=dyn_mstr_fur_A_03_r, loc=<0 0 0>, rot=<0 0 0 1>}
 dyn_cons_fur_A_04_r {parent=dyn_cons_fur_A_r, loc=<8.4033 1.203 0.3962>, rot=<0.1579 -0.1744 -0.0231 -0.9717>}
 dyn_mstr_fur_A_04_r {parent=dyn_cons_fur_A_04_r, loc=<0 0 0>, rot=<0 0 0 1>}
 dyn_fur_A_04_r {parent=dyn_mstr_fur_A_04_r, loc=<0 0 0>, rot=<0 0 0 1>}
 dyn_cons_fur_fur_B_r {parent=dfrm_furr_clavicle_r, loc=<-8.568 -9.801 -7.845>, rot=<-0.3679 -0.822 -0.1673 -0.4011>}
 dyn_cons_fur_B_01_r {parent=dyn_cons_fur_fur_B_r, loc=<14.1614 11.8008 -4.4683>, rot=<0.3417 -0.1739 -0.5733 -0.7241>}
 dyn_mstr_fur_B_01_r {parent=dyn_cons_fur_B_01_r, loc=<0 0 0>, rot=<0 0 0 1>}
 dyn_fur_B_01_r {parent=dyn_mstr_fur_B_01_r, loc=<0 0 0>, rot=<0 0 0 1>}
 dyn_cons_fur_B_02_r {parent=dyn_cons_fur_fur_B_r, loc=<14.8543 7.4664 0.9684>, rot=<0.5515 0.1989 -0.01 -0.8101>}
 dyn_mstr_fur_B_02_r {parent=dyn_cons_fur_B_02_r, loc=<0 0 0>, rot=<0 0 0 1>}
 dyn_fur_B_02_r {parent=dyn_mstr_fur_B_02_r, loc=<0 0 0>, rot=<0 0 0 1>}
 dyn_cons_fur_B_03_r {parent=dyn_cons_fur_fur_B_r, loc=<7.9728 0.8098 -2.9663>, rot=<0.1009 0.0618 0.0554 -0.9914>}
 dyn_mstr_fur_B_03_r {parent=dyn_cons_fur_B_03_r, loc=<0 0 0>, rot=<0 0 0 1>}
 dyn_fur_B_03_r {parent=dyn_mstr_fur_B_03_r, loc=<0 0 0>, rot=<0 0 0 1>}
 dyn_cons_fur_B_04_r {parent=dyn_cons_fur_fur_B_r, loc=<4.2904 0.9658 -5.6349>, rot=<0.1201 0.0163 -0.0052 -0.9926>}
 dyn_mstr_fur_B_04_r {parent=dyn_cons_fur_B_04_r, loc=<0 0 0>, rot=<0 0 0 1>}
 dyn_fur_B_04_r {parent=dyn_mstr_fur_B_04_r, loc=<0 0 0>, rot=<0 0 0 1>}
 dyn_cons_fur_B_05_r {parent=dyn_cons_fur_fur_B_r, loc=<10.3573 -0.2015 0.2429>, rot=<0.02 0.003 0.1495 -0.9886>}
 dyn_mstr_fur_B_05_r {parent=dyn_cons_fur_B_05_r, loc=<0 0 0>, rot=<0 0 0 1>}
 dyn_fur_B_05_r {parent=dyn_mstr_fur_B_05_r, loc=<0 0 0>, rot=<0 0 0 1>}
 dyn_cons_fur_back_01 {parent=spine_05, loc=<10.4676 -11.0041 3.3892>, rot=<0.3819 -0.9155 -0.0056 -0.1267>}
 dyn_mstr_fur_back_01 {parent=dyn_cons_fur_back_01, loc=<0 0 0>, rot=<0 0 0 1>}
 dyn_fur_back_01 {parent=dyn_mstr_fur_back_01, loc=<0 0 0>, rot=<0 0 0 1>}
 dyn_cons_fur_back_02 {parent=spine_05, loc=<19.7507 -7.8099 0.8958>, rot=<0.9215 -0.3756 0.0212 -0.097>}
 dyn_mstr_fur_back_02 {parent=dyn_cons_fur_back_02, loc=<0 0 0>, rot=<0 0 0 1>}
 dyn_fur_back_02 {parent=dyn_mstr_fur_back_02, loc=<0 0 0>, rot=<0 0 0 1>}
 dyn_cons_fur_back_03 {parent=spine_05, loc=<20.8338 -7.4951 4.3522>, rot=<0.97 -0.2358 -0.0494 -0.0314>}
 dyn_mstr_fur_back_03 {parent=dyn_cons_fur_back_03, loc=<0 0 0>, rot=<0 0 0 1>}
 dyn_fur_back_03 {parent=dyn_mstr_fur_back_03, loc=<0 0 0>, rot=<0 0 0 1>}
 dyn_cons_fur_back_04 {parent=spine_05, loc=<11.8122 -12.1993 -1.1946>, rot=<0.7423 -0.6651 -0.0668 -0.0464>}
 dyn_mstr_fur_back_04 {parent=dyn_cons_fur_back_04, loc=<0 0 0>, rot=<0 0 0 1>}
 dyn_fur_back_04 {parent=dyn_mstr_fur_back_04, loc=<0 0 0>, rot=<0 0 0 1>}
 dyn_cons_fur_back_05 {parent=spine_05, loc=<8.731 -12.4634 -4.5961>, rot=<0.3848 -0.9197 0.0478 0.0616>}
 dyn_mstr_fur_back_05 {parent=dyn_cons_fur_back_05, loc=<0 0 0>, rot=<0 0 0 1>}
 dyn_fur_back_05 {parent=dyn_mstr_fur_back_05, loc=<0 0 0>, rot=<0 0 0 1>}
 dyn_cons_fur_back_06 {parent=spine_05, loc=<14.7981 -8.3099 4.7841>, rot=<0.79 -0.6126 -0.0147 0.0189>}
 dyn_mstr_fur_back_06 {parent=dyn_cons_fur_back_06, loc=<0 0 0>, rot=<0 0 0 1>}
 dyn_fur_back_06 {parent=dyn_mstr_fur_back_06, loc=<0 0 0>, rot=<0 0 0 1>}
 dyn_cons_fur_back_07 {parent=spine_05, loc=<21.5202 -5.815 -3.5597>, rot=<0.9649 -0.2498 -0.0353 -0.0735>}
 dyn_mstr_fur_back_07 {parent=dyn_cons_fur_back_07, loc=<0 0 0>, rot=<0 0 0 1>}
 dyn_fur_back_07 {parent=dyn_mstr_fur_back_07, loc=<0 0 0>, rot=<0 0 0 1>}
 dyn_cons_fur_back_08 {parent=spine_05, loc=<11.3596 -9.6135 9.3252>, rot=<0.3338 -0.9199 0.18 -0.1003>}
 dyn_mstr_fur_back_08 {parent=dyn_cons_fur_back_08, loc=<0 0 0>, rot=<0 0 0 1>}
 dyn_fur_back_08 {parent=dyn_mstr_fur_back_08, loc=<0 0 0>, rot=<0 0 0 1>}
 dyn_cons_fur_A_back_09 {parent=spine_05, loc=<15.6225 -10.8562 -5.3941>, rot=<0.9145 -0.3927 0.0623 0.0752>}
 dyn_mstr_fur_A_back_09 {parent=dyn_cons_fur_A_back_09, loc=<0 0 0>, rot=<0 0 0 1>}
 dyn_fur_A_back_09 {parent=dyn_mstr_fur_A_back_09, loc=<0 0 0>, rot=<0 0 0 1>}
 deform_pec_aim_base_l {parent=spine_04, loc=<7.8732 12.6388 -2.4123>, rot=<0.0102 -0.4602 0.3979 -0.7936>}
 deform_pec_aim_l {parent=deform_pec_aim_base_l, loc=<0 0 0>, rot=<0 0 0 1>}
 deform_pec_base_l {parent=deform_pec_aim_l, loc=<10.766 0 0>, rot=<0 0 0 1>}
 deform_pec_l {parent=deform_pec_base_l, loc=<0.055 0 0>, rot=<0 0 0 1>}
 deform_pec_aim_base_r {parent=spine_04, loc=<7.8727 12.6388 2.4123>, rot=<-0.4602 -0.0102 0.7936 0.3979>}
 deform_pec_aim_r {parent=deform_pec_aim_base_r, loc=<0 0 0>, rot=<0 0 0 1>}
 deform_pec_base_r {parent=deform_pec_aim_r, loc=<-10.7662 -0.0006 -0.0007>, rot=<0 0 0 1>}
 deform_pec_r {parent=deform_pec_base_r, loc=<0 0 0>, rot=<0 0 0 1>}
 dyn_mstr_zipper {parent=spine_04, loc=<5.5826 12.9725 0>, rot=<0.0294 -0.7065 -0.0294 -0.7065>}
 dyn_zipper {parent=dyn_mstr_zipper, loc=<0 0.4401 -0.3623>, rot=<0.707 0.0117 -0.707 -0.0117>}
 dyn_mstr_skirt_back {parent=spine_01, loc=<2.9336 -10.0729 0>, rot=<0.1673 -0.9859 0 0>}
 dyn_skirt_back_01 {parent=dyn_mstr_skirt_back, loc=<0 0 0>, rot=<0 0 0 -1>}
 dyn_skirt_back_02 {parent=dyn_skirt_back_01, loc=<14.7411 0 0>, rot=<0 0 0 1>}
 dyn_skirt_back_03 {parent=dyn_skirt_back_02, loc=<14.7408 0 0>, rot=<0 0 0 1>}
 dyn_skirt_back_01_l {parent=dyn_skirt_back_03, loc=<4.605 0.254 3.9337>, rot=<0.0533 0.0147 -0.0155 -0.9984>}
 dyn_skirt_back_02_l {parent=dyn_skirt_back_01_l, loc=<14.5062 0 0>, rot=<0 0 0 1>}
 dyn_skirt_back_01_r {parent=dyn_skirt_back_03, loc=<4.6051 0.254 -3.9337>, rot=<-0.0533 -0.0147 -0.0155 -0.9984>}
 dyn_skirt_back_02_r {parent=dyn_skirt_back_01_r, loc=<14.5061 0 0>, rot=<0 0 0 1>}
 dyn_mstr_skirt_back_side_l {parent=spine_01, loc=<2.2501 -8.442 -6.7875>, rot=<0.1703 -0.9419 -0.2879 -0.03>}
 dyn_skirt_back_side_01_l {parent=dyn_mstr_skirt_back_side_l, loc=<0 0 0>, rot=<0 0 0 1>}
 dyn_skirt_back_side_02_l {parent=dyn_skirt_back_side_01_l, loc=<15.002 0 0>, rot=<0 0 0 1>}
 dyn_skirt_back_side_03_l {parent=dyn_skirt_back_side_02_l, loc=<15.002 0 0>, rot=<0 0 0.0127 -0.9999>}
 dyn_skirt_back_side_04_l {parent=dyn_skirt_back_side_03_l, loc=<13.326 0.2521 0>, rot=<0 0 -0.0494 -0.9988>}
 dyn_mstr_skirt_back_side_r {parent=spine_01, loc=<2.2504 -8.442 6.7875>, rot=<0.1703 -0.9419 0.2879 0.03>}
 dyn_skirt_back_side_01_r {parent=dyn_mstr_skirt_back_side_r, loc=<0 0 0>, rot=<0 0 0 1>}
 dyn_skirt_back_side_02_r {parent=dyn_skirt_back_side_01_r, loc=<15.0022 0.0001 0>, rot=<0 0 0 1>}
 dyn_skirt_back_side_03_r {parent=dyn_skirt_back_side_02_r, loc=<15.002 0.0001 0>, rot=<0 0 0.0032 -1>}
 dyn_skirt_back_side_04_r {parent=dyn_skirt_back_side_03_r, loc=<13.3284 0 0>, rot=<0 0 -0.0399 -0.9992>}
 dfrm_mstr_belt_front {parent=spine_01, loc=<7.6943 0.6541 13.2614>, rot=<-0.4028 -0.3809 0.6925 0.4616>}
 dfrm_belt_front_01 {parent=dfrm_mstr_belt_front, loc=<0 0 0>, rot=<0 0 0 1>}
 dfrm_belt_front_02 {parent=dfrm_belt_front_01, loc=<4.5493 0 0>, rot=<-0.0474 0.0302 -0.1711 -0.9837>}
 dfrm_belt_front_03 {parent=dfrm_belt_front_02, loc=<4.5505 0 0>, rot=<0.0138 0.0156 -0.2319 -0.9725>}
 dfrm_belt_front_04 {parent=dfrm_belt_front_03, loc=<4.5515 0 0>, rot=<-0.0384 -0.0005 -0.156 -0.987>}
 dfrm_belt_front_05 {parent=dfrm_belt_front_04, loc=<4.5525 0 0>, rot=<-0.0348 -0.0118 -0.0742 -0.9966>}
 dfrm_belt_front_06 {parent=dfrm_belt_front_05, loc=<6.8236 0 0>, rot=<-0.0162 -0.0139 -0.0414 -0.9989>}
 dfrm_belt_front_07 {parent=dfrm_belt_front_06, loc=<4.5489 0 0>, rot=<-0.0456 -0.0159 -0.0573 -0.9972>}
 dfrm_belt_front_08 {parent=dfrm_belt_front_07, loc=<4.5501 0 0>, rot=<-0.0468 -0.0286 -0.1054 -0.9929>}
 dfrm_belt_front_09 {parent=dfrm_belt_front_08, loc=<4.558 0 0>, rot=<-0.0129 -0.0647 -0.2139 -0.9746>}
 dfrm_belt_front_10 {parent=dfrm_belt_front_09, loc=<4.558 0 0>, rot=<0.0874 -0.0482 -0.2231 -0.9697>}
 dfrm_belt_front_11 {parent=dfrm_belt_front_10, loc=<4.5297 0 0>, rot=<-0.0746 -0.0047 -0.1409 -0.9872>}
 dfrm_mstr_belt_back {parent=spine_01, loc=<7.6734 0.6445 13.1858>, rot=<0.548 -0.641 0.3321 -0.4225>}
 dfrm_belt_back_01 {parent=dfrm_mstr_belt_back, loc=<0 0 0>, rot=<0 0 0 1>}
 dfrm_belt_back_02 {parent=dfrm_belt_back_01, loc=<4.6016 0 0>, rot=<0.0388 -0.0284 -0.1637 -0.9853>}
 dfrm_belt_back_03 {parent=dfrm_belt_back_02, loc=<4.6719 0 0>, rot=<-0.0106 -0.0128 -0.2087 -0.9778>}
 dfrm_belt_back_04 {parent=dfrm_belt_back_03, loc=<4.6617 0 0>, rot=<0.0069 0.0047 -0.122 -0.9925>}
 dfrm_belt_back_05 {parent=dfrm_belt_back_04, loc=<4.6005 0 0>, rot=<0.0564 0.0188 -0.114 -0.9917>}
 dfrm_belt_back_06 {parent=dfrm_belt_back_05, loc=<4.6046 0 0>, rot=<0.0771 0.0135 -0.0451 -0.9959>}
 dfrm_belt_back_07 {parent=dfrm_belt_back_06, loc=<4.5868 0 0>, rot=<0.0119 0.036 -0.0998 -0.9943>}
 dfrm_belt_back_08 {parent=dfrm_belt_back_07, loc=<4.6191 0 0>, rot=<0.0902 0.0463 -0.1059 -0.9892>}
 dfrm_belt_back_09 {parent=dfrm_belt_back_08, loc=<4.6077 0 0>, rot=<-0.0435 0.0532 -0.1295 -0.9892>}
 dfrm_belt_back_010 {parent=dfrm_belt_back_09, loc=<4.6532 0 0>, rot=<-0.0545 0.0352 -0.1684 -0.9836>}
 dfrm_belt_back_011 {parent=dfrm_belt_back_010, loc=<4.6357 0 0>, rot=<-0.0087 0.0172 -0.1574 -0.9874>}
 dfrm_belt_back_012 {parent=dfrm_belt_back_011, loc=<4.6104 0 0>, rot=<0.0291 0.043 -0.1662 -0.9847>}
 dyn_mstr_skirt_side_l {parent=pelvis, loc=<7.816 -5.8219 -12.1248>, rot=<0.1183 -0.7994 -0.5859 -0.0612>}
 dyn_skirt_side_01_l {parent=dyn_mstr_skirt_side_l, loc=<0 0 0>, rot=<0 0 0 1>}
 dyn_skirt_side_02_l {parent=dyn_skirt_side_01_l, loc=<15.5415 0 0>, rot=<0 0 0.0273 -0.9996>}
 dyn_skirt_side_03_l {parent=dyn_skirt_side_02_l, loc=<15.5415 0 0>, rot=<0 0 0 1>}
 dyn_skirt_side_04_l {parent=dyn_skirt_side_03_l, loc=<12.3015 0 0>, rot=<0 0 0 1>}
 dyn_skirt_side_05_l {parent=dyn_skirt_side_04_l, loc=<12.3015 0 0>, rot=<0 0 0 1>}
 dyn_mstr_skirt_front_l {parent=pelvis, loc=<5.3611 6.0381 -11.2733>, rot=<-0.108 0.4648 0.8757 0.0734>}
 dyn_skirt_front_01_l {parent=dyn_mstr_skirt_front_l, loc=<0 0 0>, rot=<0 0 0 1>}
 dyn_skirt_front_02_l {parent=dyn_skirt_front_01_l, loc=<15.8439 -0.0057 0>, rot=<0 0 0.0154 -0.9999>}
 dyn_skirt_front_03_l {parent=dyn_skirt_front_02_l, loc=<15.8439 0 0>, rot=<0 0 0 1>}
 dyn_skirt_front_04_l {parent=dyn_skirt_front_03_l, loc=<17.2611 0 0>, rot=<0 0 0 1>}
 dyn_skirt_front_05_l {parent=dyn_skirt_front_04_l, loc=<17.2611 0 0>, rot=<0 0 0 1>}
 dyn_mstr_skirt_front_side_l {parent=pelvis, loc=<6.5889 0.197 -13.5715>, rot=<-0.1165 0.69 0.7093 0.0849>}
 dyn_skirt_front_side_01_l {parent=dyn_mstr_skirt_front_side_l, loc=<0 0 0>, rot=<0 0 0 1>}
 dyn_skirt_front_side_02_l {parent=dyn_skirt_front_side_01_l, loc=<15.7025 0.016 0>, rot=<0 0 0.0125 -0.9999>}
 dyn_skirt_front_side_03_l {parent=dyn_skirt_front_side_02_l, loc=<15.7025 0 0>, rot=<0 0 0 1>}
 dyn_skirt_front_side_04_l {parent=dyn_skirt_front_side_03_l, loc=<16.7771 0.0009 0>, rot=<0 0 -0.0067 -1>}
 dyn_skirt_front_side_05_l {parent=dyn_skirt_front_side_04_l, loc=<16.7771 0 0>, rot=<0 0 0 1>}
 dyn_mstr_skirt_front_r {parent=pelvis, loc=<5.3611 6.0381 11.2733>, rot=<0.108 -0.4648 0.8757 0.0734>}
 dyn_skirt_front_01_r {parent=dyn_mstr_skirt_front_r, loc=<0 0 0>, rot=<0 0 0.0002 -1>}
 dyn_skirt_front_02_r {parent=dyn_skirt_front_01_r, loc=<15.8438 0 0>, rot=<0 0 0.0152 -0.9999>}
 dyn_skirt_front_03_r {parent=dyn_skirt_front_02_r, loc=<15.844 0 0>, rot=<0 0 0 1>}
 dyn_skirt_front_04_r {parent=dyn_skirt_front_03_r, loc=<17.2611 -0.0001 0.0001>, rot=<0 0 0 1>}
 dyn_skirt_front_05_r {parent=dyn_skirt_front_04_r, loc=<17.2611 0 0>, rot=<0 0 0 1>}
 dyn_mstr_skirt_front_side_r {parent=pelvis, loc=<6.589 0.197 13.5715>, rot=<0.1165 -0.69 0.7093 0.0849>}
 dyn_skirt_front_side_01_r {parent=dyn_mstr_skirt_front_side_r, loc=<0 0 0>, rot=<0 0 -0.0005 -1>}
 dyn_skirt_front_side_02_r {parent=dyn_skirt_front_side_01_r, loc=<15.7025 0 0>, rot=<0 0 0.013 -0.9999>}
 dyn_skirt_front_side_03_r {parent=dyn_skirt_front_side_02_r, loc=<15.7025 0 0>, rot=<0 0 0 -1>}
 dyn_skirt_front_side_04_r {parent=dyn_skirt_front_side_03_r, loc=<16.7771 0 0>, rot=<0 0 -0.0067 -1>}
 dyn_skirt_front_side_05_r {parent=dyn_skirt_front_side_04_r, loc=<16.7771 0 0>, rot=<0 0 0 1>}
 dyn_mstr_skirt_side_r {parent=pelvis, loc=<7.8163 -5.8219 12.1248>, rot=<0.1183 -0.7994 0.5859 0.0612>}
 dyn_skirt_side_01_r {parent=dyn_mstr_skirt_side_r, loc=<0 0 0>, rot=<0 0 0 1>}
 dyn_skirt_side_02_r {parent=dyn_skirt_side_01_r, loc=<15.5419 0 0>, rot=<0 0 0.0273 -0.9996>}
 dyn_skirt_side_03_r {parent=dyn_skirt_side_02_r, loc=<15.5415 0 0>, rot=<0 0 0 1>}
 dyn_skirt_side_04_r {parent=dyn_skirt_side_03_r, loc=<12.3015 -0.0001 0>, rot=<0 0 0 1>}
 thigh_l {parent=pelvis, loc=<-2.5693 -0.0109 -7.7501>, rot=<0.0012 0.0572 0.0155 -0.9982>}
 calf_l {parent=thigh_l, loc=<-39.0484 0 0>, rot=<0.0006 -0.0161 -0.0391 -0.9991>}
 foot_l {parent=calf_l, loc=<-42.561 0.1077 0.4292>, rot=<0.0418 -0.0455 0.0343 -0.9975>}
 ball_l {parent=foot_l, loc=<-6.1599 13.8264 0.6806>, rot=<0.0001 -0.0001 -0.7063 -0.7079>}
 calf_twist_02_l {parent=calf_l, loc=<-18.8064 -0.0967 0.0397>, rot=<0.0237 0.0095 -0.0035 -0.9997>}
 dfrm_knee_prnt_l {parent=calf_twist_02_l, loc=<6.6705 -2.7703 -6.6524>, rot=<-0.576 -0.1869 -0.0596 -0.7936>}
 dfrm_mstr_knee_tear_08_l {parent=dfrm_knee_prnt_l, loc=<0.7332 0.136 -2.6031>, rot=<-0.2017 -0.1269 0.1352 -0.9617>}
 dfrm_knee_tear_08_l {parent=dfrm_mstr_knee_tear_08_l, loc=<0 0 0>, rot=<0 0 0 1>}
 dfrm_mstr_knee_tear_09_l {parent=dfrm_knee_prnt_l, loc=<2.3506 0 0>, rot=<0.0147 0.0033 0.1667 -0.9859>}
 dfrm_knee_tear_09_l {parent=dfrm_mstr_knee_tear_09_l, loc=<0 0 0>, rot=<0 0 0 1>}
 dfrm_mstr_knee_tear_10_l {parent=dfrm_knee_prnt_l, loc=<3.1715 0.555 1.9912>, rot=<0.1741 0.081 0.082 -0.978>}
 dfrm_knee_tear_10_l {parent=dfrm_mstr_knee_tear_10_l, loc=<0 0 0>, rot=<0 0 0 1>}
 dfrm_knee_tear_07_l {parent=calf_twist_02_l, loc=<5.924 3.1589 -3.0689>, rot=<0.9657 0.1035 0.0436 0.2341>}
 dfrm_knee_tear_06_l {parent=calf_twist_02_l, loc=<7.3933 -1.686 4.7035>, rot=<0.7577 0.2883 0.0672 -0.5816>}
 calf_twist_01_l {parent=calf_l, loc=<-35.813 -0.0967 0.0397>, rot=<0.0237 0.0095 -0.0035 -0.9997>}
 deform_knee_in_base_l {parent=calf_l, loc=<1.298 -5.2957 -0.0496>, rot=<0 0 0 -1>}
 deform_knee_in_l {parent=deform_knee_in_base_l, loc=<0 0 0>, rot=<0 0 0 1>}
 deform_knee_out_base_l {parent=calf_l, loc=<0 4 0>, rot=<0 0 0 -1>}
 deform_knee_out_l {parent=deform_knee_out_base_l, loc=<0 0 0>, rot=<0 0 0 1>}
 dfrm_knee_tear_11_l {parent=thigh_l, loc=<-40.7987 -1.8401 6.2185>, rot=<0.084 -0.8081 0.5012 0.298>}
 thigh_twist_01_l {parent=thigh_l, loc=<-21.6303 0 0>, rot=<0.0013 0 0 -1>}
 dfrm_knee_tear_04_l {parent=thigh_twist_01_l, loc=<-8.3138 -0.3285 -5.0905>, rot=<0.188 -0.7127 -0.6212 -0.2661>}
 dfrm_knee_tear_03_l {parent=thigh_twist_01_l, loc=<-8.2577 4.0956 -4.2635>, rot=<-0.1678 0.5136 0.8391 -0.0627>}
 dfrm_knee_tear_02_l {parent=thigh_twist_01_l, loc=<-9.1367 6.5275 2.0098>, rot=<-0.0571 -0.0426 0.9904 0.1185>}
 dfrm_knee_tear_01_l {parent=thigh_twist_01_l, loc=<-10.2035 1.6118 5.8418>, rot=<0.1395 -0.5943 0.7897 0.061>}
 deform_groin_base_l {parent=thigh_l, loc=<-2.1908 8.1142 -1.7442>, rot=<0 0 0 -1>}
 deform_groin_l {parent=deform_groin_base_l, loc=<0 0 0>, rot=<0 0 0 1>}
 dfrm_mstr_knee_tear_05_l {parent=thigh_l, loc=<-36.6816 -3.2146 -3.9193>, rot=<-0.0409 -0.793 -0.3249 -0.5136>}
 dfrm_knee_tear_05_l {parent=dfrm_mstr_knee_tear_05_l, loc=<0 0 0>, rot=<0 0 0 1>}
 thigh_r {parent=pelvis, loc=<-2.5693 -0.0109 7.7501>, rot=<0.0572 -0.0012 0.9982 0.0155>}
 calf_r {parent=thigh_r, loc=<39.0484 0 0>, rot=<0.0006 -0.0161 -0.0391 -0.9991>}
 foot_r {parent=calf_r, loc=<42.561 -0.1077 -0.4292>, rot=<0.0418 -0.0455 0.0343 -0.9975>}
 ball_r {parent=foot_r, loc=<6.1599 -13.8264 -0.6806>, rot=<0.0001 -0.0001 -0.7063 -0.7079>}
 calf_twist_02_r {parent=calf_r, loc=<18.8064 0.0967 -0.0397>, rot=<0.0237 0.0095 -0.0035 -0.9997>}
 calf_twist_01_r {parent=calf_r, loc=<35.813 0.0967 -0.0397>, rot=<0.0237 0.0095 -0.0035 -0.9997>}
 deform_knee_in_base_r {parent=calf_r, loc=<-1.298 5.2957 0.0496>, rot=<0 0 0 -1>}
 deform_knee_in_r {parent=deform_knee_in_base_r, loc=<0 0 0>, rot=<0 0 0 1>}
 deform_knee_out_base_r {parent=calf_r, loc=<0 -4 0>, rot=<0 0 0 -1>}
 deform_knee_out_r {parent=deform_knee_out_base_r, loc=<0 0 0>, rot=<0 0 0 1>}
 thigh_twist_01_r {parent=thigh_r, loc=<21.6303 0 0>, rot=<0.0013 0 0 -1>}
 deform_groin_base_r {parent=thigh_r, loc=<2.1908 -8.1142 1.7442>, rot=<0 0 0 -1>}
 deform_groin_r {parent=deform_groin_base_r, loc=<0 0 0>, rot=<0 0 0 1>}
 deform_glute_base_l {parent=pelvis, loc=<-8.8064 -8.8222 -8.4287>, rot=<0.0275 0.0501 -0.4472 -0.8926>}
 deform_glute_l {parent=deform_glute_base_l, loc=<0 0 0>, rot=<0 0 0 1>}
 deform_glute_base_r {parent=pelvis, loc=<-8.8062 -8.8222 8.4287>, rot=<0.0501 -0.0275 0.8926 -0.4472>}
 deform_glute_r {parent=deform_glute_base_r, loc=<0 0 0>, rot=<0 0 0 1>}
 belt_end_ctrl_null {parent=pelvis, type=Null, loc=<-3.4443 -0.1883 -18.6024>, rot=<-0.5837 -0.5173 -0.4673 -0.4163>}
 belt_end_ctrl {controlType=EULER_TRANSFORM, parent=belt_end_ctrl_null, type=Control, loc=<0 0 0>, rot=<0 0 0 1>}
 dfrm_belt_front_10_ctrl_null {parent=belt_end_ctrl, type=Null, loc=<4.4796 0.6718 -0.0075>, rot=<0.0741 -0.9947 -0.0709 0.0061>}
 dfrm_belt_front_10_ctrl {controlType=EULER_TRANSFORM, parent=dfrm_belt_front_10_ctrl_null, type=Control, loc=<0 0 0>, rot=<0 0 0 1>}
 dfrm_belt_back_11_ctrl_null {parent=belt_end_ctrl, type=Null, loc=<-4.3387 1.4973 0.4351>, rot=<0.0291 0.043 -0.1662 0.9847>}
 dfrm_belt_back_11_ctrl {controlType=EULER_TRANSFORM, parent=dfrm_belt_back_11_ctrl_null, type=Control, loc=<0 0 0>, rot=<0 0 0 1>}
 dfrm_belt_front_11_ctrl {controlType=EULER_TRANSFORM, parent=belt_end_ctrl, type=Control, loc=<18.6024 -1.0187 89.6837>, rot=<0.6978 -0.7082 -0.0498 0.0952>}
 dfrm_belt_back_012_ctrl {controlType=EULER_TRANSFORM, parent=belt_end_ctrl, type=Control, loc=<18.6024 -1.0187 89.6837>, rot=<0.0705 0.0847 0.7418 0.6615>}
 dfrm_belt_back_08_ctrl_null {parent=pelvis, type=Null, loc=<1.6698 -13.3529 -8.5074>, rot=<-0.059 -0.7917 -0.2712 -0.5443>}
 dfrm_belt_back_08_ctrl {controlType=EULER_TRANSFORM, parent=dfrm_belt_back_08_ctrl_null, type=Control, loc=<0 0 0>, rot=<0 0 0 1>}
 belt_start_ctrl_null {parent=pelvis, type=Null, loc=<14.4488 -0.5178 13.1858>, rot=<0.5258 -0.6593 0.3463 -0.411>}
 belt_start_ctrl {controlType=EULER_TRANSFORM, parent=belt_start_ctrl_null, type=Control, loc=<0 0 0>, rot=<0 0 0 1>}
 dfrm_belt_front_01_ctrl {controlType=EULER_TRANSFORM, parent=belt_start_ctrl, type=Control, loc=<-13.2614 -0.6981 107.5984>, rot=<-0.7735 -0.599 0.1952 0.0695>}
 dfrm_belt_back_01_ctrl {controlType=EULER_TRANSFORM, parent=belt_start_ctrl, type=Control, loc=<-13.1858 -0.707 107.5772>, rot=<-0.1405 0.1645 -0.6197 0.7544>}
 dfrm_belt_front_08_ctrl_null {parent=pelvis, type=Null, loc=<-0.2031 9.3754 -10.8704>, rot=<-0.5365 0.1894 0.8144 -0.1142>}
 dfrm_belt_front_08_ctrl {controlType=EULER_TRANSFORM, parent=dfrm_belt_front_08_ctrl_null, type=Control, loc=<0 0 0>, rot=<0 0 0 1>}
 dfrm_belt_back_05_ctrl_null {parent=pelvis, type=Null, loc=<8.1733 -13.3107 3.5469>, rot=<0.0719 -0.8621 0.011 -0.5014>}
 dfrm_belt_back_05_ctrl {controlType=EULER_TRANSFORM, parent=dfrm_belt_back_05_ctrl_null, type=Control, loc=<0 0 0>, rot=<0 0 0 1>}
 dfrm_belt_front_03_ctrl_null {parent=pelvis, type=Null, loc=<11.9941 7.5435 10.2082>, rot=<0.4725 0.1823 -0.8456 -0.1691>}
 dfrm_belt_front_03_ctrl {controlType=EULER_TRANSFORM, parent=dfrm_belt_front_03_ctrl_null, type=Control, loc=<0 0 0>, rot=<0 0 0 1>}
 dfrm_belt_front_05_ctrl_null {parent=pelvis, type=Null, loc=<7.3825 10.3769 3.027>, rot=<-0.5008 -0.0063 0.8655 0.0066>}
 dfrm_belt_front_05_ctrl {controlType=EULER_TRANSFORM, parent=dfrm_belt_front_05_ctrl_null, type=Control, loc=<0 0 0>, rot=<0 0 0 1>}
 dfrm_belt_front_06_ctrl_null {parent=pelvis, type=Null, loc=<3.9827 10.4973 -2.8881>, rot=<-0.5125 0.0286 0.8579 -0.0211>}
 dfrm_belt_front_06_ctrl {controlType=EULER_TRANSFORM, parent=dfrm_belt_front_06_ctrl_null, type=Control, loc=<0 0 0>, rot=<0 0 0 1>}
 dfrm_belt_back_03_ctrl_null {parent=pelvis, type=Null, loc=<12.5626 -9.0174 10.3887>, rot=<0.2405 -0.8334 0.1827 -0.4628>}
 dfrm_belt_back_03_ctrl {controlType=EULER_TRANSFORM, parent=dfrm_belt_back_03_ctrl_null, type=Control, loc=<0 0 0>, rot=<0 0 0 1>}
 dyn_belt_mstr {parent=pelvis, loc=<14.4488 -0.5178 13.1858>, rot=<0.3485 -0.6183 -0.6025 -0.3651>}
 dyn_belt {parent=dyn_belt_mstr, loc=<36.4797 0 0>, rot=<0 0 0 1>}
 belt_end_origin_ctrl {controlType=EULER_TRANSFORM, parent=pelvis, type=Control, loc=<18.6024 -1.0187 89.6837>, rot=<0.0705 0.0847 0.7418 0.6615>}
 dyn_crouch_sphere {parent=root, loc=<0 0 0>, rot=<0 0 0 1>}
]
variables [
 Weight {type=float, default=1}
 back_IKspline_ctrl_bindpose {type=FRigPose, default=(Elements=,HierarchyTopologyVersion=-1,PoseHash=-1)}
 back_IKspline_thigh_l_up_pose {type=FRigPose, default=(Elements=,HierarchyTopologyVersion=-1,PoseHash=-1)}
 back_belt_body_twist_left_pose {type=FRigPose, default=(Elements=,HierarchyTopologyVersion=-1,PoseHash=-1)}
 back_spline_ctrls {
  type="TArray<FRigElementKey>"
  default="((Type=Control,Name=\"belt_start_ctrl\"),(Type=Control,Name=\"dfrm_belt_back_01_ctrl\"),(Type=Control,Name=\"dfrm_belt_back_03_ctrl\"),(Type=Control,Name=\"dfrm_belt_back_05_ctrl\"),(Type=Control,Name=\"dfrm_belt_back_08_ctrl\"),(Type=Control,Name=\"dfrm_belt_back_11 ...(truncated)"
 }
 belt_IKSpline_bindpose {type=FRigPose, default=(Elements=,HierarchyTopologyVersion=-1,PoseHash=-1)}
 belt_distance_alpha {type=double, default=0}
 belt_spline_ctrls {
  type="TArray<FRigElementKey>"
  default="((Type=Control,Name=\"belt_start_ctrl\"),(Type=Control,Name=\"dfrm_belt_back_01_ctrl\"),(Type=Control,Name=\"dfrm_belt_back_03_ctrl\"),(Type=Control,Name=\"dfrm_belt_back_05_ctrl\"),(Type=Control,Name=\"dfrm_belt_back_08_ctrl\"),(Type=Control,Name=\"dfrm_belt_back_11 ...(truncated)"
 }
 front_IKSpline_ctrl_bindpose {type=FRigPose, default=(Elements=,HierarchyTopologyVersion=-1,PoseHash=-1)}
 front_IKSpline_thigh_l_up_pose {type=FRigPose, default=(Elements=,HierarchyTopologyVersion=-1,PoseHash=-1)}
 front_spline_ctrls {
  type="TArray<FRigElementKey>"
  default="((Type=Control,Name=\"belt_start_ctrl\"),(Type=Control,Name=\"dfrm_belt_front_01_ctrl\"),(Type=Control,Name=\"dfrm_belt_front_03_ctrl\"),(Type=Control,Name=\"dfrm_belt_front_05_ctrl\"),(Type=Control,Name=\"dfrm_belt_front_06_ctrl\"),(Type=Control,Name=\"dfrm_belt_fro ...(truncated)"
 }
 is_crouching {type=bool, default=False}
 spine_twist_pos_delta_ref_alpha {type=double, default=0}
 spr_head_back_alpha {type=double, default=0}
 spr_thigh_l_back_alpha {type=double, default=0}
 spr_thigh_l_up_alpha {type=double, default=0}
 spr_thigh_r_up_alpha {type=double, default=0}
]
