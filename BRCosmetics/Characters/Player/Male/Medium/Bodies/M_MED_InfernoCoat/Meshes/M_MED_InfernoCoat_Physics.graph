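# Physics asset for the medium male InfernoCoat player body, set up for secondary motion of the coat skirt and belt.
# Skirt chains hang from master bones as capsules and boxes with limited swing and tight twist, the center back forking in two.
# Angularly free cross-links join neighboring chains at the third segment, tying panels together.
# Pelvis, leg and foot shapes, a crouch sphere and a huge ground-like root sphere provide body collision.
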PhysicsAsset /BRCosmetics/Characters/Player/Male/Medium/Bodies/M_MED_InfernoCoat/Meshes/M_MED_InfernoCoat_Physics
bodies {
 calf_l [
  {shapeType=Sphere, radius=9.3179, shapeName=calf_l_sphere, center=<-5.191 0 0>}
  {shapeType=Capsule, length=40.6168, radius=7.6303, shapeName=calf_l_capsule, center=<-19.8795 0.0003 0>, rotation=<rot 90 -0.0004 0.0003>}
 ]
 calf_r [{shapeType=Capsule, length=40.6168, radius=7.6303, shapeName=calf_l_capsule, center=<19.8795 -0.0003 0>, rotation=<rot 90 -0.0004 -179.9997>}]
 dyn_belt_mstr [{shapeType=Capsule, length=15.6583, radius=5.4387, shapeName=dyn_belt_capsule, center=<36.9722 -0.504 -2.595>}]
 dyn_crouch_sphere [{shapeType=Sphere, radius=40, shapeName=root_sphere, center=<0 -11.9806 3.336>}]
 dyn_mstr_skirt_back [{shapeType=Capsule, length=0, radius=0.6766, shapeName=dyn_mstr_skirt_front_l_capsule}]
 dyn_mstr_skirt_back_side_l [{shapeType=Capsule, length=0, radius=0.5, shapeName=dyn_mstr_skirt_front_l_capsule}]
 dyn_mstr_skirt_back_side_r [{shapeType=Capsule, length=0, radius=0.5, shapeName=dyn_mstr_skirt_front_l_capsule, rotation=<rot 0 0 -180>}]
 dyn_mstr_skirt_front_l [{shapeType=Capsule, length=0, radius=0.5, shapeName=dyn_mstr_skirt_front_l_capsule}]
 dyn_mstr_skirt_front_r [{shapeType=Capsule, length=0, radius=0.5, shapeName=dyn_mstr_skirt_front_l_capsule, rotation=<rot 0 0 -180>}]
 dyn_mstr_skirt_front_side_l [{shapeType=Capsule, length=0, radius=0.5, shapeName=dyn_mstr_skirt_front_l_capsule}]
 dyn_mstr_skirt_front_side_r [{shapeType=Capsule, length=0, radius=0.5, shapeName=dyn_mstr_skirt_front_l_capsule, rotation=<rot 0 0 -180>}]
 dyn_mstr_skirt_side_l [{shapeType=Capsule, length=0, radius=0.5, shapeName=dyn_mstr_skirt_front_l_capsule}]
 dyn_mstr_skirt_side_r [{shapeType=Capsule, length=0, radius=0.5, shapeName=dyn_mstr_skirt_front_l_capsule, rotation=<rot 0 0 -180>}]
 dyn_skirt_back_01 [{shapeType=Capsule, length=10.8584, radius=0.8395, shapeName=dyn_skirt_front_01_l_capsule, center=<15.2509 0 0>}]
 dyn_skirt_back_01_l [{shapeType=Box, extentX=14.2264, extentY=1.5673, extentZ=8.9828, shapeName=dyn_skirt_front_01_l_box, center=<7.086 0 0>}]
 dyn_skirt_back_01_r [{shapeType=Box, extentX=14.2264, extentY=1.5673, extentZ=8.9828, shapeName=dyn_skirt_front_01_l_box, center=<7.086 0 0>}]
 dyn_skirt_back_02 [{shapeType=Box, extentX=14.7061, extentY=1.5673, extentZ=11.2046, shapeName=dyn_skirt_front_01_l_box, center=<8.2712 0 0>}]
 dyn_skirt_back_02_l [{shapeType=Capsule, length=6.0813, radius=0.8395, shapeName=dyn_skirt_front_01_l_capsule, center=<15.2509 0 0>}]
 dyn_skirt_back_02_r [{shapeType=Capsule, length=6.0813, radius=0.8395, shapeName=dyn_skirt_front_01_l_capsule, center=<15.2509 0 0>}]
 dyn_skirt_back_03 [{shapeType=Box, extentX=4.2476, extentY=1.5673, extentZ=11.2046, shapeName=dyn_skirt_front_01_l_box, center=<2.5204 0 0>}]
 dyn_skirt_back_side_01_l [{shapeType=Capsule, length=10.8584, radius=0.8395, shapeName=dyn_skirt_front_01_l_capsule, center=<15.2509 0 0>}]
 dyn_skirt_back_side_01_r [{shapeType=Capsule, length=10.8584, radius=0.8395, shapeName=dyn_skirt_front_01_l_capsule, center=<15.2509 0 0>}]
 dyn_skirt_back_side_02_l [{shapeType=Box, extentX=14.2264, extentY=1.5673, extentZ=11.2046, shapeName=dyn_skirt_front_01_l_box, center=<7.9235 0 0>}]
 dyn_skirt_back_side_02_r [{shapeType=Box, extentX=14.2264, extentY=1.5673, extentZ=11.2046, shapeName=dyn_skirt_front_01_l_box, center=<7.9235 0 0>}]
 dyn_skirt_back_side_03_l [{shapeType=Box, extentX=14.2264, extentY=1.5673, extentZ=11.2046, shapeName=dyn_skirt_front_01_l_box, center=<7.9235 0 0>}]
 dyn_skirt_back_side_03_r [{shapeType=Box, extentX=14.2264, extentY=1.5673, extentZ=11.2046, shapeName=dyn_skirt_front_01_l_box, center=<7.9235 0 0>}]
 dyn_skirt_back_side_04_l [{shapeType=Box, extentX=11.7494, extentY=1.5673, extentZ=9.227, shapeName=dyn_skirt_front_01_l_box, center=<7.9235 0 0>}]
 dyn_skirt_back_side_04_r [{shapeType=Box, extentX=18.4899, extentY=1.5673, extentZ=10.908, shapeName=dyn_skirt_front_01_l_box, center=<10.9904 0 0>}]
 dyn_skirt_front_01_l [{shapeType=Capsule, length=5.3882, radius=1.5327, shapeName=dyn_skirt_front_01_l_capsule, center=<15.2509 0 0>}]
 dyn_skirt_front_01_r [{shapeType=Box, extentX=16.9638, extentY=1.3957, extentZ=7.978, shapeName=dyn_skirt_front_01_l_box, center=<8.7831 0 0.9>}]
 dyn_skirt_front_02_l [{shapeType=Capsule, length=6.0813, radius=0.8395, shapeName=dyn_skirt_front_01_l_capsule, center=<15.2509 0 0>}]
 dyn_skirt_front_02_r [{shapeType=Box, extentX=14.4177, extentY=1.4934, extentZ=8.1186, shapeName=dyn_skirt_front_01_l_box, center=<8.7831 0 0.9>}]
 dyn_skirt_front_03_l [{shapeType=Box, extentX=17.1043, extentY=1.2442, extentZ=8.1186, shapeName=dyn_skirt_front_01_l_box, center=<8.7831 0 -0.9>}]
 dyn_skirt_front_03_r [{shapeType=Box, extentX=17.1043, extentY=1.2442, extentZ=8.1186, shapeName=dyn_skirt_front_01_l_box, center=<8.7831 0 0.9>}]
 dyn_skirt_front_04_l [{shapeType=Box, extentX=17.1043, extentY=1.2442, extentZ=10.0604, shapeName=dyn_skirt_front_01_l_box, center=<8.7831 0 -0.9>}]
 dyn_skirt_front_04_r [{shapeType=Box, extentX=17.1043, extentY=1.2442, extentZ=9.4244, shapeName=dyn_skirt_front_01_l_box, center=<8.7831 0 1.16>}]
 dyn_skirt_front_05_l [{shapeType=Box, extentX=23.3363, extentY=1.2442, extentZ=10.4364, shapeName=dyn_skirt_front_01_l_box, center=<12.146 0 -0.9>}]
 dyn_skirt_front_05_r [{shapeType=Box, extentX=23.3363, extentY=1.2442, extentZ=10.8796, shapeName=dyn_skirt_front_01_l_box, center=<12.146 0 1.16>}]
 dyn_skirt_front_side_01_l [{shapeType=Capsule, length=5.8902, radius=1.0306, shapeName=dyn_skirt_front_01_l_capsule, center=<15.2509 0 0>}]
 dyn_skirt_front_side_01_r [{shapeType=Capsule, length=6.0813, radius=0.8395, shapeName=dyn_skirt_front_01_l_capsule, center=<15.2509 0 0>}]
 dyn_skirt_front_side_02_l [{shapeType=Capsule, length=6.0813, radius=0.8395, shapeName=dyn_skirt_front_01_l_capsule, center=<15.2509 0 0>}]
 dyn_skirt_front_side_02_r [{shapeType=Capsule, length=6.0813, radius=0.8395, shapeName=dyn_skirt_front_01_l_capsule, center=<15.2509 0 0>}]
 dyn_skirt_front_side_03_l [{shapeType=Box, extentX=15.492, extentY=2.7767, extentZ=10.4439, shapeName=dyn_skirt_front_01_l_box, center=<8.1943 0 0>}]
 dyn_skirt_front_side_03_r [{shapeType=Box, extentX=15.492, extentY=2.7767, extentZ=10.4439, shapeName=dyn_skirt_front_01_l_box, center=<8.1943 0 0>}]
 dyn_skirt_front_side_04_l [{shapeType=Box, extentX=15.9121, extentY=2.7767, extentZ=11.0844, shapeName=dyn_skirt_front_01_l_box, center=<7.3162 0 -0.0653>}]
 dyn_skirt_front_side_04_r [{shapeType=Box, extentX=15.9121, extentY=2.7767, extentZ=10.4439, shapeName=dyn_skirt_front_01_l_box, center=<7.3162 0 -0.0653>}]
 dyn_skirt_front_side_05_l [{shapeType=Box, extentX=15.9121, extentY=2.7767, extentZ=12.1893, shapeName=dyn_skirt_front_01_l_box, center=<7.3162 0 -0.0653>}]
 dyn_skirt_front_side_05_r [{
   shapeType=Box
   extentX=15.9121
   extentY=2.7767
   extentZ=18.1982
   shapeName=dyn_skirt_front_01_l_box
   center=<7.3162 0 4.5>
   rotation=<rot -0.0001 0 5>
  }]
 dyn_skirt_side_01_l [{shapeType=Capsule, length=6.0813, radius=0.8395, shapeName=dyn_skirt_front_01_l_capsule, center=<15.2509 0 0>}]
 dyn_skirt_side_01_r [{shapeType=Capsule, length=6.0813, radius=0.8395, shapeName=dyn_skirt_front_01_l_capsule, center=<15.2509 0 0>}]
 dyn_skirt_side_02_l [{shapeType=Box, extentX=14.7512, extentY=1.5673, extentZ=7.7832, shapeName=dyn_skirt_front_01_l_box, center=<7.9235 0 0>}]
 dyn_skirt_side_02_r [{shapeType=Box, extentX=14.7512, extentY=1.5673, extentZ=7.7832, shapeName=dyn_skirt_front_01_l_box, center=<7.9235 0 0>}]
 dyn_skirt_side_03_l [{shapeType=Box, extentX=12.2852, extentY=1.5673, extentZ=9.7836, shapeName=dyn_skirt_front_01_l_box, center=<6.2807 0 0>}]
 dyn_skirt_side_03_r [{shapeType=Box, extentX=14.7512, extentY=1.5673, extentZ=9.7836, shapeName=dyn_skirt_front_01_l_box, center=<7.9235 0 0>}]
 dyn_skirt_side_04_l [{shapeType=Box, extentX=12.2852, extentY=1.5673, extentZ=9.7836, shapeName=dyn_skirt_front_01_l_box, center=<6.2807 0 0>}]
 dyn_skirt_side_04_r [{shapeType=Box, extentX=12.2852, extentY=1.5673, extentZ=9.7836, shapeName=dyn_skirt_front_01_l_box, center=<9.2402 0 0>}]
 dyn_skirt_side_05_l [{shapeType=Box, extentX=17.6681, extentY=1.5673, extentZ=10.8485, shapeName=dyn_skirt_front_01_l_box, center=<9.4695 0 0>}]
 foot_l [{shapeType=Capsule, length=16.8483, radius=6.819, shapeName=foot_l_capsule, center=<-4.4801 7.576 0>, rotation=<rot 0 -0.0001 -90>}]
 foot_r [{shapeType=Capsule, length=16.8483, radius=6.819, shapeName=foot_l_capsule, center=<4.4801 -7.576 0>, rotation=<rot 0 -0.0001 90>}]
 pelvis [{shapeType=Capsule, length=7.5706, radius=11.9493, shapeName=pelvis_capsule, center=<-3.565 -1.3857 -1.0466>}]
 root [
  {shapeType=Sphere, radius=497.7262, shapeName=root_sphere, center=<0 0 -481.8383>}
  {shapeType=Box, extentX=7.5, extentY=7.5, extentZ=7.5, shapeName=root_box}
 ]
 thigh_l [
  {shapeType=Sphere, radius=9.1967, shapeName=thigh_l_sphere, center=<-37.9507 0 0>}
  {shapeType=Capsule, length=29.7382, radius=7.9307, shapeName=thigh_l_capsule, center=<-16.4261 0 0.2205>, rotation=<rot 90 3.3723 3.3725>}
 ]
 thigh_r [
  {shapeType=Sphere, radius=9.1967, shapeName=thigh_l_sphere, center=<37.9507 0 0>}
  {shapeType=Capsule, length=29.7382, radius=7.9307, shapeName=thigh_l_capsule, center=<16.4261 0 -0.2205>, rotation=<rot 90 179.156 -0.8438>}
 ]
}
constraints [
 {bone1Name=dyn_skirt_front_02_l, bone2Name=dyn_skirt_front_01_l, swing1LimitDegrees=25, swing2LimitDegrees=15, twistLimitDegrees=2}
 {bone1Name=dyn_skirt_front_01_l, bone2Name=dyn_mstr_skirt_front_l, swing1LimitDegrees=15, swing2LimitDegrees=10, twistLimitDegrees=2}
 {bone1Name=dyn_skirt_front_side_02_l, bone2Name=dyn_skirt_front_side_01_l, swing1LimitDegrees=25, swing2LimitDegrees=10, twistLimitDegrees=2}
 {bone1Name=dyn_skirt_front_side_01_l, bone2Name=dyn_mstr_skirt_front_side_l, swing1LimitDegrees=5, swing2LimitDegrees=10, twistLimitDegrees=2}
 {bone1Name=dyn_skirt_side_02_l, bone2Name=dyn_skirt_side_01_l, swing1LimitDegrees=25, swing2LimitDegrees=10, twistLimitDegrees=2}
 {bone1Name=dyn_skirt_side_01_l, bone2Name=dyn_mstr_skirt_side_l, swing1LimitDegrees=5, swing2LimitDegrees=10, twistLimitDegrees=2}
 {bone1Name=dyn_skirt_back_side_02_l, bone2Name=dyn_skirt_back_side_01_l, swing1LimitDegrees=25, swing2LimitDegrees=15, twistLimitDegrees=2}
 {bone1Name=dyn_skirt_back_side_01_l, bone2Name=dyn_mstr_skirt_back_side_l, swing1LimitDegrees=5, swing2LimitDegrees=10, twistLimitDegrees=2}
 {bone1Name=dyn_skirt_front_01_r, bone2Name=dyn_mstr_skirt_front_r, swing1LimitDegrees=35, swing2LimitDegrees=25, twistLimitDegrees=2}
 {bone1Name=dyn_skirt_front_02_r, bone2Name=dyn_skirt_front_01_r, swing1LimitDegrees=25, swing2LimitDegrees=20, twistLimitDegrees=2}
 {bone1Name=dyn_skirt_front_side_02_r, bone2Name=dyn_skirt_front_side_01_r, swing1LimitDegrees=25, swing2LimitDegrees=15, twistLimitDegrees=2}
 {
  bone1Name=dyn_skirt_front_side_01_r
  bone2Name=dyn_mstr_skirt_front_side_r
  swing1LimitDegrees=30
  swing2LimitDegrees=15
  twistLimitDegrees=2
 }
 {bone1Name=dyn_skirt_side_02_r, bone2Name=dyn_skirt_side_01_r, swing1LimitDegrees=25, swing2LimitDegrees=15, twistLimitDegrees=2}
 {bone1Name=dyn_skirt_side_01_r, bone2Name=dyn_mstr_skirt_side_r, swing1LimitDegrees=30, swing2LimitDegrees=10, twistLimitDegrees=2}
 {bone1Name=dyn_skirt_back_side_02_r, bone2Name=dyn_skirt_back_side_01_r, swing1LimitDegrees=25, swing2LimitDegrees=15, twistLimitDegrees=2}
 {bone1Name=dyn_skirt_back_side_01_r, bone2Name=dyn_mstr_skirt_back_side_r, swing1LimitDegrees=30, swing2LimitDegrees=10, twistLimitDegrees=2}
 {bone1Name=dyn_skirt_back_02, bone2Name=dyn_skirt_back_01, swing1LimitDegrees=30, swing2LimitDegrees=10, twistLimitDegrees=2}
 {bone1Name=dyn_skirt_back_01, bone2Name=dyn_mstr_skirt_back, swing1LimitDegrees=5, swing2LimitDegrees=25, twistLimitDegrees=2}
 {bone1Name=dyn_skirt_front_03_l, bone2Name=dyn_skirt_front_02_l, swing1LimitDegrees=35, swing2LimitDegrees=5, twistLimitDegrees=2}
 {bone1Name=dyn_skirt_front_03_r, bone2Name=dyn_skirt_front_02_r, swing1LimitDegrees=25, swing2LimitDegrees=5, twistLimitDegrees=2}
 {bone1Name=dyn_skirt_front_side_03_l, bone2Name=dyn_skirt_front_side_02_l, swing1LimitDegrees=35, swing2LimitDegrees=5, twistLimitDegrees=2}
 {bone1Name=dyn_skirt_front_side_03_r, bone2Name=dyn_skirt_front_side_02_r, swing1LimitDegrees=15, swing2LimitDegrees=25, twistLimitDegrees=2}
 {bone1Name=dyn_skirt_side_03_l, bone2Name=dyn_skirt_side_02_l, swing1LimitDegrees=35, swing2LimitDegrees=5, twistLimitDegrees=2}
 {bone1Name=dyn_skirt_side_03_r, bone2Name=dyn_skirt_side_02_r, swing1LimitDegrees=15, swing2LimitDegrees=25, twistLimitDegrees=2}
 {bone1Name=dyn_skirt_back_side_03_l, bone2Name=dyn_skirt_back_side_02_l, swing1LimitDegrees=25, swing2LimitDegrees=5, twistLimitDegrees=2}
 {bone1Name=dyn_skirt_back_side_03_r, bone2Name=dyn_skirt_back_side_02_r, swing1LimitDegrees=25, swing2LimitDegrees=5, twistLimitDegrees=2}
 {bone1Name=dyn_skirt_back_02_l, bone2Name=dyn_skirt_back_01_l, swing1LimitDegrees=25, swing2LimitDegrees=15, twistLimitDegrees=2}
 {bone1Name=dyn_skirt_back_02_r, bone2Name=dyn_skirt_back_01_r, swing1LimitDegrees=25, swing2LimitDegrees=15, twistLimitDegrees=2}
 {bone1Name=dyn_skirt_back_03, bone2Name=dyn_skirt_back_02, swing1LimitDegrees=35, swing2LimitDegrees=15, twistLimitDegrees=2}
 {bone1Name=dyn_skirt_back_01_l, bone2Name=dyn_skirt_back_03, swing1LimitDegrees=35, swing2LimitDegrees=15, twistLimitDegrees=2}
 {bone1Name=dyn_skirt_back_01_r, bone2Name=dyn_skirt_back_03, swing1LimitDegrees=35, swing2LimitDegrees=15, twistLimitDegrees=2}
 {
  bone1Name=dyn_skirt_front_03_l
  bone2Name=dyn_skirt_front_side_03_l
  swing1LimitDegrees=45
  swing1Motion=Free
  swing2LimitDegrees=45
  swing2Motion=Free
  twistLimitDegrees=45
  twistMotion=Free
 }
 {
  bone1Name=dyn_skirt_front_side_03_l
  bone2Name=dyn_skirt_side_03_l
  swing1LimitDegrees=45
  swing1Motion=Free
  swing2LimitDegrees=45
  swing2Motion=Free
  twistLimitDegrees=45
  twistMotion=Free
 }
 {
  bone1Name=dyn_skirt_side_03_l
  bone2Name=dyn_skirt_back_side_03_l
  swing1LimitDegrees=45
  swing1Motion=Free
  swing2LimitDegrees=45
  swing2Motion=Free
  twistLimitDegrees=45
  twistMotion=Free
 }
 {
  bone1Name=dyn_skirt_back_side_03_l
  bone2Name=dyn_skirt_back_03
  swing1LimitDegrees=45
  swing1Motion=Free
  swing2LimitDegrees=45
  swing2Motion=Free
  twistLimitDegrees=45
  twistMotion=Free
 }
 {
  bone1Name=dyn_skirt_front_03_r
  bone2Name=dyn_skirt_front_side_03_r
  swing1LimitDegrees=45
  swing1Motion=Free
  swing2LimitDegrees=45
  swing2Motion=Free
  twistLimitDegrees=45
  twistMotion=Free
 }
 {
  bone1Name=dyn_skirt_front_side_03_r
  bone2Name=dyn_skirt_side_03_r
  swing1LimitDegrees=45
  swing1Motion=Free
  swing2LimitDegrees=45
  swing2Motion=Free
  twistLimitDegrees=45
  twistMotion=Free
 }
 {
  bone1Name=dyn_skirt_side_03_r
  bone2Name=dyn_skirt_back_side_03_r
  swing1LimitDegrees=45
  swing1Motion=Free
  swing2LimitDegrees=45
  swing2Motion=Free
  twistLimitDegrees=45
  twistMotion=Free
 }
 {
  bone1Name=dyn_skirt_back_side_03_r
  bone2Name=dyn_skirt_back_03
  swing1LimitDegrees=45
  swing1Motion=Free
  swing2LimitDegrees=45
  swing2Motion=Free
  twistLimitDegrees=45
  twistMotion=Free
 }
 {bone1Name=dyn_skirt_front_04_l, bone2Name=dyn_skirt_front_03_l, swing1LimitDegrees=25, swing2LimitDegrees=5, twistLimitDegrees=2}
 {bone1Name=dyn_skirt_front_04_r, bone2Name=dyn_skirt_front_03_r, swing1LimitDegrees=25, swing2LimitDegrees=5, twistLimitDegrees=2}
 {bone1Name=dyn_skirt_front_side_04_l, bone2Name=dyn_skirt_front_side_03_l, swing1LimitDegrees=25, swing2LimitDegrees=5, twistLimitDegrees=2}
 {bone1Name=dyn_skirt_front_side_04_r, bone2Name=dyn_skirt_front_side_03_r, swing1LimitDegrees=25, swing2LimitDegrees=5, twistLimitDegrees=2}
 {bone1Name=dyn_skirt_side_04_l, bone2Name=dyn_skirt_side_03_l, swing1LimitDegrees=25, swing2LimitDegrees=5, twistLimitDegrees=2}
 {bone1Name=dyn_skirt_side_04_r, bone2Name=dyn_skirt_side_03_r, swing1LimitDegrees=25, swing2LimitDegrees=5, twistLimitDegrees=2}
 {bone1Name=dyn_skirt_back_side_04_l, bone2Name=dyn_skirt_back_side_03_l, swing1LimitDegrees=25, swing2LimitDegrees=5, twistLimitDegrees=2}
 {bone1Name=dyn_skirt_back_side_04_r, bone2Name=dyn_skirt_back_side_03_r, swing1LimitDegrees=25, swing2LimitDegrees=5, twistLimitDegrees=2}
 {bone1Name=dyn_skirt_front_05_l, bone2Name=dyn_skirt_front_04_l, swing1LimitDegrees=25, swing2LimitDegrees=5, twistLimitDegrees=2}
 {bone1Name=dyn_skirt_front_05_r, bone2Name=dyn_skirt_front_04_r, swing1LimitDegrees=25, swing2LimitDegrees=5, twistLimitDegrees=2}
 {bone1Name=dyn_skirt_front_side_05_l, bone2Name=dyn_skirt_front_side_04_l, swing1LimitDegrees=25, swing2LimitDegrees=5, twistLimitDegrees=2}
 {bone1Name=dyn_skirt_front_side_05_r, bone2Name=dyn_skirt_front_side_04_r, swing1LimitDegrees=25, swing2LimitDegrees=5, twistLimitDegrees=2}
 {bone1Name=dyn_skirt_side_05_l, bone2Name=dyn_skirt_side_04_l, swing1LimitDegrees=25, swing2LimitDegrees=5, twistLimitDegrees=2}
 {
  bone1Name=calf_r
  bone2Name=thigh_r
  swing1LimitDegrees=45
  swing1Motion=Free
  swing2LimitDegrees=45
  swing2Motion=Free
  twistLimitDegrees=45
  twistMotion=Free
 }
 {bone1Name=dyn_belt_mstr, bone2Name=pelvis, swing1LimitDegrees=25, swing2LimitDegrees=2, twistLimitDegrees=2}
 {
  bone1Name=foot_r
  bone2Name=thigh_r
  swing1LimitDegrees=45
  swing1Motion=Free
  swing2LimitDegrees=45
  swing2Motion=Free
  twistLimitDegrees=45
  twistMotion=Free
 }
 {
  bone1Name=thigh_r
  bone2Name=pelvis
  swing1LimitDegrees=45
  swing1Motion=Free
  swing2LimitDegrees=45
  swing2Motion=Free
  twistLimitDegrees=45
  twistMotion=Free
 }
]
