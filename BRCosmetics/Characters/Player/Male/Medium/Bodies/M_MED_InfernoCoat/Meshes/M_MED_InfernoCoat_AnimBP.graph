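# Post-process style AnimBP for the InfernoCoat male body cosmetic, adding secondary motion to the copied parent pose.
# Deform and main Control Rigs feed zipper, knee-tear and fur AnimDynamics, with constraints blending in-between fur bones.
# RigidBody under a per-update gravity override follows, then a belt Control Rig driven by spine twist and crouch.
# EventGraph also drives material world-position-offset from three upper-body regions and switches cloth wind on while skydiving.
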AnimBlueprint /BRCosmetics/Characters/Player/Male/Medium/Bodies/M_MED_InfernoCoat/Meshes/M_MED_InfernoCoat_AnimBP
target_skeleton=&/M_MED_InfernoCoat_Skeleton
graphs [
 AnimGraph {nodes [
   n1 = Get fur_alpha() -> (fur_alpha:double)
   n2 = Control Rig(Source=Copy Pose From Mesh() {bUseAttachedParent=true}) {
    alphaBoolBlend {blendOption=Linear}
    controlRigAssetReference {blueprintRigClass=&/M_MED_InfernoCoat_Deform_CtrlRig.M_MED_InfernoCoat_Deform_CtrlRig_C}
   }
   n3 = Control Rig!(Source=@n2, Alpha, is_crouching:bool=Get bIsCrouching(), is_skydiving:bool=Get bIsSkyDiving()) -> (Pose) {
    alphaBoolBlend {blendOption=Linear}
    controlRigAssetReference {blueprintRigClass=&/M_MED_InfernoCoat_CtrlRig.M_MED_InfernoCoat_CtrlRig_C}
   }
   n4 = "Constraint - dyn_mstr_zipper (PPPPP)"!(
    ComponentPose=Local To Component(LocalPose=@n3)
    Alpha=1
    ConstraintWeights_0=0.4556
    ConstraintWeights_1:float=0.2692
    ConstraintWeights_2:float=0.1323
    ConstraintWeights_3:float=0.1259
    ConstraintWeights_4:float=0.017
   ) -> (Pose) {
    alphaBoolBlend {blendOption=Linear}
    boneToModify {boneName=dyn_mstr_zipper}
    constraintSetup [
     {targetBone {boneName=spine_04}}
     {targetBone {boneName=spine_03}}
     {targetBone {boneName=spine_02}}
     {targetBone {boneName=spine_05}}
     {targetBone {boneName=spine_01}}
    ]
    constraintWeights [1,1,1,1,1]
   }
   n5 = Anim Dynamics - Bone: dyn_zipper(ComponentPose=@n4, Alpha=1) {
    angularDampingOverride=0.98
    angularSpringConstant=500
    bAngularSpring=true
    bOverrideAngularDamping=true
    bUseGravityOverride=true
    gravityScale=0.35
    simulationSpace=RootRelative
    alphaBoolBlend {blendOption=Linear}
    boundBone {boneName=dyn_zipper}
    componentLinearAccScale=<1 1 1>
    componentLinearVelScale=<1 1 1>
    gravityOverride {z=-980}
    physicsBodyDefinitions [{
      boundBone {boneName=dyn_zipper}
      boxExtents=<4 4 4>
      constraintSetup {angularLimitsMax=<0 10 0>, angularLimitsMin=<0 -10 -30>, angularTarget=<1 0 0>}
      localJointOffset=<1.8 0 0>
     }]
   }
   n6 = "Transform (Modify) Bone - Bone: dfrm_knee_prnt_l"(ComponentPose=@n5, Alpha=1, Translation="0,0,0", Rotation="10,20,0", Scale="1,1,1") {rotationMode=BMM_Additive, alphaBoolBlend {blendOption=Linear}, boneToModify {boneName=dfrm_knee_prnt_l}, rotation {pitch=10, yaw=20}}
   n7 = "Transform (Modify) Bone - Bone: dfrm_mstr_knee_tear_09_l"(ComponentPose=@n6, Alpha=1, Translation="0,0,0", Rotation="15,0,0", Scale="1,1,1") {rotationMode=BMM_Additive, alphaBoolBlend {blendOption=Linear}, boneToModify {boneName=dfrm_mstr_knee_tear_09_l}, rotation {pitch=15}}
   n8 = "Transform (Modify) Bone - Bone: dfrm_mstr_knee_tear_05_l"(ComponentPose=@n7, Alpha=1, Translation="0,0,0", Rotation="0,30,0", Scale="1,1,1") {rotationMode=BMM_Additive, alphaBoolBlend {blendOption=Linear}, boneToModify {boneName=dfrm_mstr_knee_tear_05_l}, rotation {yaw=30}}
   n9 = Anim Dynamics - Bone: dfrm_knee_tear_01_l(ComponentPose=@n8, Alpha=1, ExternalForce="0,0,0") {
    angularDampingOverride=0.9999
    angularSpringConstant=2500
    bAngularSpring=true
    bOverrideAngularDamping=true
    bUseGravityOverride=true
    simulationSpace=BoneRelative
    alphaBoolBlend {blendOption=Linear}
    boundBone {boneName=dfrm_knee_tear_01_l}
    componentAppliedLinearAccClamp=<5000 5000 5000>
    componentLinearAccScale=<1.2 1.2 1.2>
    componentLinearVelScale=<1.2 1.2 1.2>
    gravityOverride {z=-980}
    physicsBodyDefinitions [{
      sphereCollisionRadius=0
      boundBone {boneName=dfrm_knee_tear_01_l}
      boxExtents=<14 14 14>
      constraintSetup {angularLimitsMax=<2.5 2.5 20>, angularLimitsMin=<-2.5 -2.5 -20>, angularTarget=<1 0 0>}
      localJointOffset=<0.5 0 0>
     }]
    relativeSpaceBone {boneName=root}
   }
   n10 = Anim Dynamics - Bone: dfrm_knee_tear_02_l(ComponentPose=@n9, Alpha=1, ExternalForce="0,0,0") {
    angularDampingOverride=0.9999
    angularSpringConstant=2500
    bAngularSpring=true
    bOverrideAngularDamping=true
    bUseGravityOverride=true
    simulationSpace=BoneRelative
    alphaBoolBlend {blendOption=Linear}
    boundBone {boneName=dfrm_knee_tear_02_l}
    componentAppliedLinearAccClamp=<5000 5000 5000>
    componentLinearAccScale=<1.2 1.2 1.2>
    componentLinearVelScale=<1.2 1.2 1.2>
    gravityOverride {z=-980}
    physicsBodyDefinitions [{
      sphereCollisionRadius=0
      boundBone {boneName=dfrm_knee_tear_02_l}
      boxExtents=<13 13 13>
      constraintSetup {angularLimitsMax=<2.5 2.5 20>, angularLimitsMin=<-2.5 -2.5 -20>, angularTarget=<1 0 0>}
      localJointOffset=<0.65 0 0>
     }]
    relativeSpaceBone {boneName=root}
   }
   n11 = Anim Dynamics - Bone: dfrm_knee_tear_03_l(ComponentPose=@n10, Alpha=1, ExternalForce="0,0,0") {
    angularDampingOverride=0.9999
    angularSpringConstant=2500
    bAngularSpring=true
    bOverrideAngularDamping=true
    bUseGravityOverride=true
    simulationSpace=BoneRelative
    alphaBoolBlend {blendOption=Linear}
    boundBone {boneName=dfrm_knee_tear_03_l}
    componentAppliedLinearAccClamp=<5000 5000 5000>
    componentLinearAccScale=<1.2 1.2 1.2>
    componentLinearVelScale=<1.2 1.2 1.2>
    gravityOverride {z=-980}
    physicsBodyDefinitions [{
      sphereCollisionRadius=0
      boundBone {boneName=dfrm_knee_tear_03_l}
      boxExtents=<12 12 12>
      constraintSetup {angularLimitsMax=<2.5 2.5 10>, angularLimitsMin=<-2.5 -2.5 -10>, angularTarget=<1 0 0>}
      localJointOffset=<0.7 0 0>
     }]
    relativeSpaceBone {boneName=root}
   }
   n12 = Anim Dynamics - Bone: dfrm_knee_tear_11_l(ComponentPose=@n11, Alpha=1, ExternalForce="0,0,0") {
    angularDampingOverride=0.9999
    angularSpringConstant=2500
    bAngularSpring=true
    bOverrideAngularDamping=true
    bUseGravityOverride=true
    simulationSpace=BoneRelative
    alphaBoolBlend {blendOption=Linear}
    boundBone {boneName=dfrm_knee_tear_11_l}
    componentAppliedLinearAccClamp=<5000 5000 5000>
    componentLinearAccScale=<1.2 1.2 1.2>
    componentLinearVelScale=<1.2 1.2 1.2>
    gravityOverride {z=-980}
    physicsBodyDefinitions [{
      sphereCollisionRadius=0
      boundBone {boneName=dfrm_knee_tear_11_l}
      boxExtents=<12 12 12>
      constraintSetup {angularLimitsMax=<2.5 2.5 10>, angularLimitsMin=<-2.5 -2.5 -10>, angularTarget=<1 0 0>}
      localJointOffset=<0.8 0 0>
     }]
    relativeSpaceBone {boneName=root}
   }
   n13 = Anim Dynamics - Bone: dfrm_knee_tear_07_l(ComponentPose=@n12, Alpha=1, ExternalForce="0,0,0") {
    angularDampingOverride=0.9999
    angularSpringConstant=2500
    bAngularSpring=true
    bOverrideAngularDamping=true
    bUseGravityOverride=true
    simulationSpace=BoneRelative
    alphaBoolBlend {blendOption=Linear}
    boundBone {boneName=dfrm_knee_tear_07_l}
    componentAppliedLinearAccClamp=<5000 5000 5000>
    componentLinearAccScale=<1.2 1.2 1.2>
    componentLinearVelScale=<1.2 1.2 1.2>
    gravityOverride {z=-980}
    physicsBodyDefinitions [{
      sphereCollisionRadius=0
      boundBone {boneName=dfrm_knee_tear_07_l}
      boxExtents=<12 12 12>
      constraintSetup {angularLimitsMax=<2.5 2.5 10>, angularLimitsMin=<-2.5 -2.5 -10>, angularTarget=<1 0 0>}
      localJointOffset=<1 1 0>
     }]
    relativeSpaceBone {boneName=root}
   }
   n14 = Anim Dynamics - Bone: dfrm_knee_tear_06_l(ComponentPose=@n13, Alpha=1, ExternalForce="0,0,0") {
    angularDampingOverride=0.9999
    angularSpringConstant=2500
    bAngularSpring=true
    bOverrideAngularDamping=true
    bUseGravityOverride=true
    simulationSpace=BoneRelative
    alphaBoolBlend {blendOption=Linear}
    boundBone {boneName=dfrm_knee_tear_06_l}
    componentAppliedLinearAccClamp=<5000 5000 5000>
    componentLinearAccScale=<1.2 1.2 1.2>
    componentLinearVelScale=<1.2 1.2 1.2>
    gravityOverride {z=-980}
    physicsBodyDefinitions [{
      sphereCollisionRadius=0
      boundBone {boneName=dfrm_knee_tear_06_l}
      boxExtents=<11 11 11>
      constraintSetup {angularLimitsMax=<2.5 2.5 10>, angularLimitsMin=<-2.5 -2.5 -10>, angularTarget=<1 0 0>}
      localJointOffset=<1 1 0>
     }]
    relativeSpaceBone {boneName=root}
   }
   n15 = "Constraint - dyn_cons_fur_B_04_r (PPPPPPPP)"!(
    ComponentPose=@n14
    Alpha=1
    ConstraintWeights_0=0.2657
    ConstraintWeights_1:float=0.2377
    ConstraintWeights_2:float=0.1975
    ConstraintWeights_3:float=0.1393
    ConstraintWeights_4:float=0.0997
    ConstraintWeights_5:float=0.0273
    ConstraintWeights_6:float=0.017
    ConstraintWeights_7:float=0.0158
   ) -> (Pose) {
    alphaBoolBlend {blendOption=Linear}
    boneToModify {boneName=dyn_cons_fur_B_04_r}
    constraintSetup [
     {targetBone {boneName=deform_upperarm_up_r}}
     {targetBone {boneName=dfrm_clavicle_r}}
     {targetBone {boneName=upperarm_twist_01_r}}
     {targetBone {boneName=deform_upperarm_fr_r}}
     {targetBone {boneName=spine_05}}
     {targetBone {boneName=spine_04}}
     {targetBone {boneName=upperarm_twist_02_r}}
     {targetBone {boneName=deform_pec_r}}
    ]
    constraintWeights [1,1,1,1,1,1,1,1]
   }
   n16 = Anim Dynamics - Bone: dyn_mstr_fur_A_l(ComponentPose=@n15, Alpha=@n1, ExternalForce="0,0,0") {
    angularDampingOverride=1
    angularSpringConstant=2400
    bAngularSpring=true
    bOverrideAngularDamping=true
    bUseGravityOverride=true
    gravityScale=0.35
    simulationSpace=RootRelative
    alphaBoolBlend {blendOption=Linear}
    boundBone {boneName=dyn_mstr_fur_A_l}
    componentAppliedLinearAccClamp=<500 500 500>
    componentLinearAccScale=<0.1 0.1 0.1>
    componentLinearVelScale=<0.9 0.9 0.9>
    gravityOverride {z=-980}
    physicsBodyDefinitions [{
      boundBone {boneName=dyn_mstr_fur_A_l}
      boxExtents=<10 10 10>
      constraintSetup {angularLimitsMax=<1 2.5 5>, angularLimitsMin=<-1 -7.5 -5>, angularTarget=<1 0 0>}
      localJointOffset=<0.2 0 0>
     }]
    relativeSpaceBone {boneName=spine_05}
   }
   n17 = Anim Dynamics - Bone: dyn_fur_A_03_l(ComponentPose=@n16, Alpha=@n1, ExternalForce="0,0,0") {
    angularDampingOverride=1
    angularSpringConstant=850
    bAngularSpring=true
    bOverrideAngularDamping=true
    bUseGravityOverride=true
    gravityScale=0.35
    simulationSpace=RootRelative
    alphaBoolBlend {blendOption=Linear}
    boundBone {boneName=dyn_fur_A_03_l}
    componentAppliedLinearAccClamp=<500 500 500>
    componentLinearAccScale=<0.1 0.1 0.1>
    componentLinearVelScale=<0.9 0.9 0.9>
    gravityOverride {z=-980}
    physicsBodyDefinitions [{
      boundBone {boneName=dyn_fur_A_03_l}
      boxExtents=<4.25 4.25 4.25>
      constraintSetup {angularLimitsMax=<1 3.5 10>, angularLimitsMin=<-1 -3.5 -30>, angularTarget=<1 0 0>}
      localJointOffset=<0.2 0 0>
     }]
    relativeSpaceBone {boneName=spine_05}
   }
   n18 = Anim Dynamics - Bone: dyn_fur_A_02_l(ComponentPose=@n17, Alpha=@n1, ExternalForce="0,0,0") {
    angularDampingOverride=1
    angularSpringConstant=975
    bAngularSpring=true
    bOverrideAngularDamping=true
    bUseGravityOverride=true
    gravityScale=0.35
    simulationSpace=RootRelative
    alphaBoolBlend {blendOption=Linear}
    boundBone {boneName=dyn_fur_A_02_l}
    componentAppliedLinearAccClamp=<500 500 500>
    componentLinearAccScale=<0.1 0.1 0.1>
    componentLinearVelScale=<0.9 0.9 0.9>
    gravityOverride {z=-980}
    physicsBodyDefinitions [{
      boundBone {boneName=dyn_fur_A_02_l}
      boxExtents=<4.5 4.5 4.5>
      constraintSetup {angularLimitsMax=<1 6 15>, angularLimitsMin=<-1 -5 -15>, angularTarget=<1 0 0>}
      localJointOffset=<0.3 0 0>
     }]
    relativeSpaceBone {boneName=spine_05}
   }
   n19 = Anim Dynamics - Bone: dyn_mstr_fur_B_l(ComponentPose=@n18, Alpha=@n1, ExternalForce="0,0,0") {
    angularDampingOverride=1
    angularSpringConstant=2800
    bAngularSpring=true
    bOverrideAngularDamping=true
    bUseGravityOverride=true
    gravityScale=0.35
    simulationSpace=RootRelative
    alphaBoolBlend {blendOption=Linear}
    boundBone {boneName=dyn_mstr_fur_B_l}
    componentAppliedLinearAccClamp=<500 500 500>
    componentLinearAccScale=<0.1 0.1 0.1>
    componentLinearVelScale=<0.9 0.9 0.9>
    gravityOverride {z=-980}
    physicsBodyDefinitions [{
      boundBone {boneName=dyn_mstr_fur_B_l}
      boxExtents=<10 10 10>
      constraintSetup {angularLimitsMax=<1 2.5 2.5>, angularLimitsMin=<-1 -2.5 -2.5>, angularTarget=<1 0 0>}
      localJointOffset=<0.16 0 0>
     }]
    relativeSpaceBone {boneName=spine_05}
   }
   n20 = Anim Dynamics - Bone: dyn_fur_B_02_l(ComponentPose=@n19, Alpha=@n1, ExternalForce="0,0,0") {
    angularDampingOverride=1
    angularSpringConstant=1500
    bAngularSpring=true
    bOverrideAngularDamping=true
    bUseGravityOverride=true
    gravityScale=0.35
    simulationSpace=RootRelative
    alphaBoolBlend {blendOption=Linear}
    boundBone {boneName=dyn_fur_B_02_l}
    componentAppliedLinearAccClamp=<500 500 500>
    componentLinearAccScale=<0.1 0.1 0.1>
    componentLinearVelScale=<0.9 0.9 0.9>
    gravityOverride {z=-980}
    physicsBodyDefinitions [{
      boundBone {boneName=dyn_fur_B_02_l}
      boxExtents=<5 5 5>
      constraintSetup {angularLimitsMax=<1 10 20>, angularLimitsMin=<-1 -10 -20>, angularTarget=<1 0 0>}
      localJointOffset=<0.35 0 0>
     }]
    relativeSpaceBone {boneName=spine_05}
   }
   n21 = Anim Dynamics - Bone: dyn_fur_B_03_l(ComponentPose=@n20, Alpha=@n1, ExternalForce="0,0,0") {
    angularDampingOverride=1
    angularSpringConstant=1050
    bAngularSpring=true
    bOverrideAngularDamping=true
    bUseGravityOverride=true
    gravityScale=0.35
    simulationSpace=RootRelative
    alphaBoolBlend {blendOption=Linear}
    boundBone {boneName=dyn_fur_B_03_l}
    componentAppliedLinearAccClamp=<500 500 500>
    componentLinearAccScale=<0.1 0.1 0.1>
    componentLinearVelScale=<0.9 0.9 0.9>
    gravityOverride {z=-980}
    physicsBodyDefinitions [{
      boundBone {boneName=dyn_fur_B_03_l}
      boxExtents=<4 4 4>
      constraintSetup {angularLimitsMax=<1 10 20>, angularLimitsMin=<-1 -10 -20>, angularTarget=<1 0 0>}
      localJointOffset=<0.35 0 0>
     }]
    relativeSpaceBone {boneName=spine_05}
   }
   n22 = Anim Dynamics - Bone: dyn_fur_B_04_l(ComponentPose=@n21, Alpha=@n1, ExternalForce="0,0,0") {
    angularDampingOverride=1
    angularSpringConstant=950
    bAngularSpring=true
    bOverrideAngularDamping=true
    bUseGravityOverride=true
    gravityScale=0.35
    simulationSpace=RootRelative
    alphaBoolBlend {blendOption=Linear}
    boundBone {boneName=dyn_fur_B_04_l}
    componentAppliedLinearAccClamp=<500 500 500>
    componentLinearAccScale=<0.1 0.1 0.1>
    componentLinearVelScale=<0.9 0.9 0.9>
    gravityOverride {z=-980}
    physicsBodyDefinitions [{
      boundBone {boneName=dyn_fur_B_04_l}
      boxExtents=<3.5 3.5 3.5>
      constraintSetup {angularLimitsMax=<1 3 20>, angularLimitsMin=<-1 -3 -20>, angularTarget=<1 0 0>}
      localJointOffset=<0.35 0 0>
     }]
    relativeSpaceBone {boneName=spine_05}
   }
   n23 = Anim Dynamics - Bone: dyn_fur_B_05_l(ComponentPose=@n22, Alpha=@n1, ExternalForce="0,0,0") {
    angularDampingOverride=1
    angularSpringConstant=950
    bAngularSpring=true
    bOverrideAngularDamping=true
    bUseGravityOverride=true
    gravityScale=0.35
    simulationSpace=RootRelative
    alphaBoolBlend {blendOption=Linear}
    boundBone {boneName=dyn_fur_B_05_l}
    componentAppliedLinearAccClamp=<500 500 500>
    componentLinearAccScale=<0.1 0.1 0.1>
    componentLinearVelScale=<0.9 0.9 0.9>
    gravityOverride {z=-980}
    physicsBodyDefinitions [{
      boundBone {boneName=dyn_fur_B_05_l}
      boxExtents=<3.5 3.5 3.5>
      constraintSetup {angularLimitsMax=<1 3 20>, angularLimitsMin=<-1 -3 -30>, angularTarget=<1 0 0>}
      localJointOffset=<0.45 0 0>
     }]
    relativeSpaceBone {boneName=spine_05}
   }
   n24 = Anim Dynamics - Bone: dyn_fur_B_07_l(ComponentPose=@n23, Alpha=@n1, ExternalForce="0,0,0") {
    angularDampingOverride=1
    angularSpringConstant=1000
    bAngularSpring=true
    bOverrideAngularDamping=true
    bUseGravityOverride=true
    gravityScale=0.35
    simulationSpace=RootRelative
    alphaBoolBlend {blendOption=Linear}
    boundBone {boneName=dyn_fur_B_07_l}
    componentAppliedLinearAccClamp=<500 500 500>
    componentLinearAccScale=<0.1 0.1 0.1>
    componentLinearVelScale=<0.9 0.9 0.9>
    gravityOverride {z=-980}
    physicsBodyDefinitions [{
      boundBone {boneName=dyn_fur_B_07_l}
      boxExtents=<4 4 4>
      constraintSetup {angularLimitsMax=<1 3.5 10>, angularLimitsMin=<-1 -3.5 -30>, angularTarget=<1 0 0>}
      localJointOffset=<0.25 0 0>
     }]
    relativeSpaceBone {boneName=spine_05}
   }
   n25 = Anim Dynamics - Bone: dyn_fur_back_05(ComponentPose=@n24, Alpha=@n1, ExternalForce="0,0,0") {
    angularDampingOverride=1
    angularSpringConstant=1600
    bAngularSpring=true
    bOverrideAngularDamping=true
    bUseGravityOverride=true
    gravityScale=0.35
    simulationSpace=RootRelative
    alphaBoolBlend {blendOption=Linear}
    boundBone {boneName=dyn_fur_back_05}
    componentAppliedLinearAccClamp=<500 500 500>
    componentLinearAccScale=<0.1 0.1 0.1>
    componentLinearVelScale=<0.9 0.9 0.9>
    gravityOverride {z=-980}
    physicsBodyDefinitions [{
      boundBone {boneName=dyn_fur_back_05}
      boxExtents=<5.5 5.5 5.5>
      constraintSetup {angularLimitsMax=<1 10 20>, angularLimitsMin=<-1 -10 -20>, angularTarget=<1 0 0>}
      localJointOffset=<0.3 0 0>
     }]
    relativeSpaceBone {boneName=spine_05}
   }
   n26 = Anim Dynamics - Bone: dyn_fur_back_06(ComponentPose=@n25, Alpha=@n1, ExternalForce="0,0,0") {
    angularDampingOverride=1
    angularSpringConstant=1550
    bAngularSpring=true
    bOverrideAngularDamping=true
    bUseGravityOverride=true
    gravityScale=0.35
    simulationSpace=RootRelative
    alphaBoolBlend {blendOption=Linear}
    boundBone {boneName=dyn_fur_back_06}
    componentAppliedLinearAccClamp=<500 500 500>
    componentLinearAccScale=<0.1 0.1 0.1>
    componentLinearVelScale=<0.9 0.9 0.9>
    gravityOverride {z=-980}
    physicsBodyDefinitions [{
      boundBone {boneName=dyn_fur_back_06}
      boxExtents=<5.5 5.5 5.5>
      constraintSetup {angularLimitsMax=<1 10 20>, angularLimitsMin=<-1 -10 -20>, angularTarget=<1 0 0>}
      localJointOffset=<0.275 0 0>
     }]
    relativeSpaceBone {boneName=spine_05}
   }
   n27 = Anim Dynamics - Bone: dyn_fur_back_08(ComponentPose=@n26, Alpha=@n1, ExternalForce="0,0,0") {
    angularDampingOverride=1
    angularSpringConstant=850
    bAngularSpring=true
    bOverrideAngularDamping=true
    bUseGravityOverride=true
    gravityScale=0.35
    simulationSpace=RootRelative
    alphaBoolBlend {blendOption=Linear}
    boundBone {boneName=dyn_fur_back_08}
    componentAppliedLinearAccClamp=<500 500 500>
    componentLinearAccScale=<0.1 0.1 0.1>
    componentLinearVelScale=<0.9 0.9 0.9>
    gravityOverride {z=-980}
    physicsBodyDefinitions [{
      boundBone {boneName=dyn_fur_back_08}
      boxExtents=<4.5 4.5 4.5>
      constraintSetup {angularLimitsMax=<1 5 10>, angularLimitsMin=<-1 -5 -30>, angularTarget=<1 0 0>}
      localJointOffset=<0.25 0 0>
     }]
    relativeSpaceBone {boneName=spine_05}
   }
   n28 = Anim Dynamics - Bone: dyn_fur_A_back_09(ComponentPose=@n27, Alpha=@n1, ExternalForce="0,0,0") {
    angularDampingOverride=1
    angularSpringConstant=1700
    bAngularSpring=true
    bOverrideAngularDamping=true
    bUseGravityOverride=true
    gravityScale=0.35
    simulationSpace=RootRelative
    alphaBoolBlend {blendOption=Linear}
    boundBone {boneName=dyn_fur_A_back_09}
    componentAppliedLinearAccClamp=<500 500 500>
    componentLinearAccScale=<0.1 0.1 0.1>
    componentLinearVelScale=<0.9 0.9 0.9>
    gravityOverride {z=-980}
    physicsBodyDefinitions [{
      boundBone {boneName=dyn_fur_A_back_09}
      boxExtents=<6 6 6>
      constraintSetup {angularLimitsMax=<1 10 20>, angularLimitsMin=<-1 -10 -20>, angularTarget=<1 0 0>}
      localJointOffset=<0.3 0 0>
     }]
    relativeSpaceBone {boneName=spine_05}
   }
   n29 = Anim Dynamics - Bone: dyn_cons_fur_A_r(ComponentPose=@n28, Alpha=@n1, ExternalForce="0,0,0") {
    angularDampingOverride=1
    angularSpringConstant=2400
    bAngularSpring=true
    bOverrideAngularDamping=true
    bUseGravityOverride=true
    gravityScale=0.35
    simulationSpace=RootRelative
    alphaBoolBlend {blendOption=Linear}
    boundBone {boneName=dyn_cons_fur_A_r}
    componentAppliedLinearAccClamp=<500 500 500>
    componentLinearAccScale=<0.1 0.1 0.1>
    componentLinearVelScale=<0.9 0.9 0.9>
    gravityOverride {z=-980}
    physicsBodyDefinitions [{
      boundBone {boneName=dyn_cons_fur_A_r}
      boxExtents=<10 10 10>
      constraintSetup {angularLimitsMax=<1 7.5 2.5>, angularLimitsMin=<-1 -2.5 -5>, angularTarget=<1 0 0>}
      localJointOffset=<0.2 0 0>
     }]
    relativeSpaceBone {boneName=spine_05}
   }
   n30 = Anim Dynamics - Bone: dyn_fur_A_01_r(ComponentPose=@n29, Alpha=@n1, ExternalForce="0,0,0") {
    angularDampingOverride=1
    angularSpringConstant=950
    bAngularSpring=true
    bOverrideAngularDamping=true
    bUseGravityOverride=true
    gravityScale=0.35
    simulationSpace=RootRelative
    alphaBoolBlend {blendOption=Linear}
    boundBone {boneName=dyn_fur_A_01_r}
    componentAppliedLinearAccClamp=<500 500 500>
    componentLinearAccScale=<0.1 0.1 0.1>
    componentLinearVelScale=<0.9 0.9 0.9>
    gravityOverride {z=-980}
    physicsBodyDefinitions [{
      boundBone {boneName=dyn_fur_A_01_r}
      boxExtents=<3.5 3.5 3.5>
      constraintSetup {angularLimitsMax=<1 3 20>, angularLimitsMin=<-1 -3 -30>, angularTarget=<1 0 0>}
      localJointOffset=<0.45 0 0>
     }]
    relativeSpaceBone {boneName=spine_05}
   }
   n31 = Anim Dynamics - Bone: dyn_fur_A_03_r(ComponentPose=@n30, Alpha=@n1, ExternalForce="0,0,0") {
    angularDampingOverride=1
    angularSpringConstant=950
    bAngularSpring=true
    bOverrideAngularDamping=true
    bUseGravityOverride=true
    gravityScale=0.35
    simulationSpace=RootRelative
    alphaBoolBlend {blendOption=Linear}
    boundBone {boneName=dyn_fur_A_03_r}
    componentAppliedLinearAccClamp=<500 500 500>
    componentLinearAccScale=<0.1 0.1 0.1>
    componentLinearVelScale=<0.9 0.9 0.9>
    gravityOverride {z=-980}
    physicsBodyDefinitions [{
      boundBone {boneName=dyn_fur_A_03_r}
      boxExtents=<4.5 4.5 4.5>
      constraintSetup {angularLimitsMax=<1 3 20>, angularLimitsMin=<-1 -3 -30>, angularTarget=<1 0 0>}
      localJointOffset=<0.3 0 0>
     }]
    relativeSpaceBone {boneName=spine_05}
   }
   n32 = Anim Dynamics - Bone: dyn_cons_fur_fur_B_r(ComponentPose=@n31, Alpha=@n1, ExternalForce="0,0,0") {
    angularDampingOverride=1
    angularSpringConstant=3000
    bAngularSpring=true
    bOverrideAngularDamping=true
    bUseGravityOverride=true
    gravityScale=0.35
    simulationSpace=RootRelative
    alphaBoolBlend {blendOption=Linear}
    boundBone {boneName=dyn_cons_fur_fur_B_r}
    componentAppliedLinearAccClamp=<500 500 500>
    componentLinearAccScale=<0.1 0.1 0.1>
    componentLinearVelScale=<0.9 0.9 0.9>
    gravityOverride {z=-980}
    physicsBodyDefinitions [{
      boundBone {boneName=dyn_cons_fur_fur_B_r}
      boxExtents=<10 10 10>
      constraintSetup {angularLimitsMax=<1 2.5 0>, angularLimitsMin=<-1 -2.5 -7.5>, angularTarget=<1 0 0>}
      localJointOffset=<0.14 0 0>
     }]
    relativeSpaceBone {boneName=spine_05}
   }
   n33 = Anim Dynamics - Bone: dyn_fur_B_01_r(ComponentPose=@n32, Alpha=@n1, ExternalForce="0,0,0") {
    angularDampingOverride=1
    angularSpringConstant=1100
    bAngularSpring=true
    bOverrideAngularDamping=true
    bUseGravityOverride=true
    gravityScale=0.35
    simulationSpace=RootRelative
    alphaBoolBlend {blendOption=Linear}
    boundBone {boneName=dyn_fur_B_01_r}
    componentAppliedLinearAccClamp=<500 500 500>
    componentLinearAccScale=<0.1 0.1 0.1>
    componentLinearVelScale=<0.9 0.9 0.9>
    gravityOverride {z=-980}
    physicsBodyDefinitions [{
      boundBone {boneName=dyn_fur_B_01_r}
      boxExtents=<4.5 4.5 4.5>
      constraintSetup {angularLimitsMax=<1 3 5>, angularLimitsMin=<-1 -3 -30>, angularTarget=<1 0 0>}
      localJointOffset=<0.3 0 0>
     }]
    relativeSpaceBone {boneName=spine_05}
   }
   n34 = Anim Dynamics - Bone: dyn_fur_B_02_r(ComponentPose=@n33, Alpha=@n1, ExternalForce="0,0,0") {
    angularDampingOverride=1
    angularSpringConstant=1400
    bAngularSpring=true
    bOverrideAngularDamping=true
    bUseGravityOverride=true
    gravityScale=0.35
    simulationSpace=RootRelative
    alphaBoolBlend {blendOption=Linear}
    boundBone {boneName=dyn_fur_B_02_r}
    componentAppliedLinearAccClamp=<500 500 500>
    componentLinearAccScale=<0.1 0.1 0.1>
    componentLinearVelScale=<0.9 0.9 0.9>
    gravityOverride {z=-980}
    physicsBodyDefinitions [{
      boundBone {boneName=dyn_fur_B_02_r}
      boxExtents=<6 6 6>
      constraintSetup {angularLimitsMax=<1 3 20>, angularLimitsMin=<-1 -3 -30>, angularTarget=<1 0 0>}
      localJointOffset=<0.35 0 0>
     }]
    relativeSpaceBone {boneName=spine_05}
   }
   n35 = Anim Dynamics - Bone: dyn_fur_B_03_r(ComponentPose=@n34, Alpha=@n1, ExternalForce="0,0,0") {
    angularDampingOverride=1
    angularSpringConstant=850
    bAngularSpring=true
    bOverrideAngularDamping=true
    bUseGravityOverride=true
    gravityScale=0.35
    simulationSpace=RootRelative
    alphaBoolBlend {blendOption=Linear}
    boundBone {boneName=dyn_fur_B_03_r}
    componentAppliedLinearAccClamp=<500 500 500>
    componentLinearAccScale=<0.1 0.1 0.1>
    componentLinearVelScale=<0.9 0.9 0.9>
    gravityOverride {z=-980}
    physicsBodyDefinitions [{
      boundBone {boneName=dyn_fur_B_03_r}
      boxExtents=<4.5 4.5 4.5>
      constraintSetup {angularLimitsMax=<1 3 20>, angularLimitsMin=<-1 -3 -30>, angularTarget=<1 0 0>}
      localJointOffset=<0.25 0 0>
     }]
    relativeSpaceBone {boneName=spine_05}
   }
   n36 = Anim Dynamics - Bone: dyn_fur_B_04_r(ComponentPose=@n35, Alpha=@n1, ExternalForce="0,0,0") {
    angularDampingOverride=1
    angularSpringConstant=1500
    bAngularSpring=true
    bOverrideAngularDamping=true
    bUseGravityOverride=true
    gravityScale=0.35
    simulationSpace=RootRelative
    alphaBoolBlend {blendOption=Linear}
    boundBone {boneName=dyn_fur_B_04_r}
    componentAppliedLinearAccClamp=<500 500 500>
    componentLinearAccScale=<0.1 0.1 0.1>
    componentLinearVelScale=<0.9 0.9 0.9>
    gravityOverride {z=-980}
    physicsBodyDefinitions [{
      boundBone {boneName=dyn_fur_B_04_r}
      boxExtents=<4 4 4>
      constraintSetup {angularLimitsMax=<1 10 5>, angularLimitsMin=<-1 -10 -20>, angularTarget=<1 0 0>}
      localJointOffset=<0.35 0 0>
     }]
    relativeSpaceBone {boneName=spine_05}
   }
   n37 = "Constraint - dyn_fur_A_01_l (RR)"!(ComponentPose=@n36, Alpha=1, ConstraintWeights_0=0.5, ConstraintWeights_1:float=0.5) -> (Pose) {
    alphaBoolBlend {blendOption=Linear}
    boneToModify {boneName=dyn_fur_A_01_l}
    constraintSetup [{transformType=Rotation, targetBone {boneName=dyn_fur_A_03_l}}, {transformType=Rotation, targetBone {boneName=dyn_fur_back_07}}]
    constraintWeights [1,1]
   }
   n38 = "Constraint - dyn_fur_B_01_l (RR)"!(ComponentPose=@n37, Alpha=1, ConstraintWeights_0=0.5, ConstraintWeights_1:float=0.5) -> (Pose) {
    alphaBoolBlend {blendOption=Linear}
    boneToModify {boneName=dyn_fur_B_01_l}
    constraintSetup [{transformType=Rotation, targetBone {boneName=dyn_fur_B_02_l}}, {transformType=Rotation, targetBone {boneName=dyn_fur_back_05}}]
    constraintWeights [1,1]
   }
   n39 = "Constraint - dyn_fur_B_06_l (RR)"!(ComponentPose=@n38, Alpha=1, ConstraintWeights_0=0.5, ConstraintWeights_1:float=0.5) -> (Pose) {
    alphaBoolBlend {blendOption=Linear}
    boneToModify {boneName=dyn_fur_B_06_l}
    constraintSetup [{transformType=Rotation, targetBone {boneName=dyn_fur_B_07_l}}, {transformType=Rotation, targetBone {boneName=dyn_fur_B_05_l}}]
    constraintWeights [1,1]
   }
   n40 = "Constraint - dyn_fur_B_08_l (RR)"!(ComponentPose=@n39, Alpha=1, ConstraintWeights_0=0.65, ConstraintWeights_1:float=0.35) -> (Pose) {
    alphaBoolBlend {blendOption=Linear}
    boneToModify {boneName=dyn_fur_B_08_l}
    constraintSetup [{transformType=Rotation, targetBone {boneName=dyn_fur_B_07_l}}, {transformType=Rotation, targetBone {boneName=dyn_fur_B_05_l}}]
    constraintWeights [1,1]
   }
   n41 = "Constraint - dyn_fur_back_01 (RR)"!(ComponentPose=@n40, Alpha=1, ConstraintWeights_0=0.5, ConstraintWeights_1:float=0.5) -> (Pose) {
    alphaBoolBlend {blendOption=Linear}
    boneToModify {boneName=dyn_fur_back_01}
    constraintSetup [{transformType=Rotation, targetBone {boneName=dyn_fur_back_08}}, {transformType=Rotation, targetBone {boneName=dyn_fur_back_05}}]
    constraintWeights [1,1]
   }
   n42 = "Constraint - dyn_fur_back_02 (RR)"!(ComponentPose=@n41, Alpha=1, ConstraintWeights_0=0.5, ConstraintWeights_1:float=0.5) -> (Pose) {
    alphaBoolBlend {blendOption=Linear}
    boneToModify {boneName=dyn_fur_back_02}
    constraintSetup [{transformType=Rotation, targetBone {boneName=dyn_fur_A_back_09}}, {transformType=Rotation, targetBone {boneName=dyn_fur_back_06}}]
    constraintWeights [1,1]
   }
   n43 = "Constraint - dyn_fur_back_03 (RR)"!(ComponentPose=@n42, Alpha=1, ConstraintWeights_0=0.5, ConstraintWeights_1:float=0.5) -> (Pose) {
    alphaBoolBlend {blendOption=Linear}
    boneToModify {boneName=dyn_fur_back_03}
    constraintSetup [{transformType=Rotation, targetBone {boneName=dyn_fur_A_03_l}}, {transformType=Rotation, targetBone {boneName=dyn_fur_B_03_r}}]
    constraintWeights [1,1]
   }
   n44 = "Constraint - dyn_fur_back_04 (RR)"!(ComponentPose=@n43, Alpha=1, ConstraintWeights_0=0.3, ConstraintWeights_1:float=0.7) -> (Pose) {
    alphaBoolBlend {blendOption=Linear}
    boneToModify {boneName=dyn_fur_back_04}
    constraintSetup [{transformType=Rotation, targetBone {boneName=dyn_fur_A_back_09}}, {transformType=Rotation, targetBone {boneName=dyn_fur_back_06}}]
    constraintWeights [1,1]
   }
   n45 = "Constraint - dyn_fur_back_07 (RR)"!(ComponentPose=@n44, Alpha=1, ConstraintWeights_0=0.5, ConstraintWeights_1:float=0.5) -> (Pose) {
    alphaBoolBlend {blendOption=Linear}
    boneToModify {boneName=dyn_fur_back_07}
    constraintSetup [{transformType=Rotation, targetBone {boneName=dyn_fur_A_02_l}}, {transformType=Rotation, targetBone {boneName=dyn_fur_back_05}}]
    constraintWeights [1,1]
   }
   n46 = "Constraint - dyn_fur_A_02_r (RR)"!(ComponentPose=@n45, Alpha=1, ConstraintWeights_0=0.4, ConstraintWeights_1:float=0.6) -> (Pose) {
    alphaBoolBlend {blendOption=Linear}
    boneToModify {boneName=dyn_fur_A_02_r}
    constraintSetup [{transformType=Rotation, targetBone {boneName=dyn_fur_A_03_l}}, {transformType=Rotation, targetBone {boneName=dyn_fur_A_01_r}}]
    constraintWeights [1,1]
   }
   n47 = "Constraint - dyn_fur_A_04_r (RR)"!(ComponentPose=@n46, Alpha=1, ConstraintWeights_0=0.5, ConstraintWeights_1:float=0.5) -> (Pose) {
    alphaBoolBlend {blendOption=Linear}
    boneToModify {boneName=dyn_fur_A_04_r}
    constraintSetup [{transformType=Rotation, targetBone {boneName=dyn_fur_A_03_l}}, {transformType=Rotation, targetBone {boneName=dyn_fur_back_07}}]
    constraintWeights [1,1]
   }
   n48 = "Constraint - dyn_fur_B_05_r (RR)"!(ComponentPose=@n47, Alpha=1, ConstraintWeights_0=0.5, ConstraintWeights_1:float=0.5) -> (Pose) {
    alphaBoolBlend {blendOption=Linear}
    boneToModify {boneName=dyn_fur_B_05_r}
    constraintSetup [{transformType=Rotation, targetBone {boneName=dyn_fur_A_03_l}}, {transformType=Rotation, targetBone {boneName=dyn_fur_B_03_r}}]
    constraintWeights [1,1]
   }
   n49 = RigidBody!(ComponentPose=@n48, Alpha, OverrideWorldGravity:Vector=Get gravity_override() -> (gravity_override:Vector)) -> (Pose) {
    bClampLinearTranslationLimitToRefPose=true
    bOverrideWorldGravity=true
    overridePhysicsAsset=&/M_MED_InfernoCoat_Physics
    simulationSpace=BaseBoneSpace
    alphaBoolBlend {blendOption=Linear}
    baseBoneRef {boneName=pelvis}
    componentAppliedLinearAccClamp=<2500 2500 2500>
    componentLinearAccScale=<0.4 0.3 0.1>
    componentLinearVelScale=<0.4 0.3 0.1>
    overrideWorldGravity {z=-980}
    simSpaceSettings {maxAngularAcceleration=1000, maxAngularVelocity=3, maxLinearAcceleration=250, maxLinearVelocity=500, velocityScaleZ=0, worldAlpha=1}
   }
   n50 = Control Rig!(
    Source=Component To Local(ComponentPose=@n49)
    Alpha
    spine_twist_pos_delta_ref_alpha:double=Get spine_twist_pos_delta_ref_alpha() -> (spine_twist_pos_delta_ref_alpha:double)
    is_crouching=Get bIsCrouching()
   ) -> (Pose) {
    alphaBoolBlend {blendOption=Linear}
    controlRigAssetReference {blueprintRigClass=&/M_MED_InfernoCoat_Belt_CtrlRig.M_MED_InfernoCoat_Belt_CtrlRig_C}
   }
   Output Pose(Result=@n50) {layerGroup=None}
  ]}
 EventGraph {nodes [
   n1 = Event BlueprintUpdateAnimation()
   > n2 = Sequence!(execute) -> (then_0, then_1, then_2:Exec)
   > n3 = Gravity Override State Function DEPRECATED(CharacterPartAnimInstance=Self-Reference(), GameTimeInSeconds=GetGameTimeInSeconds())
   > Set gravity_override(gravity_override=@n3)
   Is Valid(exec=@n2.then_2, InputObject=Get Player Pawn() -> (Player Pawn:Fort Player Pawn Athena Object Reference))
   > n4 = Sequence!(execute) -> (then_0, then_1)
   > Set WPO Globals Function(
    Material=Get material_instance() -> (material_instance:Material Instance Dynamic Object Reference)
    Overlay Material=Get overlay_mid() -> (overlay_mid:Material Instance Dynamic Object Reference)
    Skeletal Mesh=Get skeletal_mesh() -> (skeletal_mesh:Skeletal Mesh Component Object Reference)
    "Local Noise Transform  Location"="Get Local Noise Transform  Location"() -> ("Local Noise Transform  Location":Vector)
    Local Noise Offset=Get Local Noise Offset()
   )
   > Set WPO Regional Function!(
    execute
    Material:Material Instance Dynamic Object Reference=Get material_instance()
    Overlay Material:Material Instance Dynamic Object Reference=Get overlay_mid()
    Skeletal Mesh:Skeletal Mesh Component Object Reference=Get skeletal_mesh()
    Velocity Transform:Name=spine_05
    Velocity Interp Speed:double=Get Velocity Interp Speed()
    Collision Transform:Name=clavicle_l
    Collision Transform Axis:EAxis Enum
    Invert Collision Transform Axis:bool=true
    Delta Time X:double=@n1.DeltaTimeX
    Previous Velocity Transform Location:"Vector (by ref)"=Get Previous Velocity Transform Location() -> (Previous Velocity Transform Location:Vector)
    Previous Velocity Transform Velocity:"Vector (by ref)"=Get Previous Velocity Transform Velocity() -> (Previous Velocity Transform Velocity:Vector)
    Region:int=1
   ) -> (then)
   > Set WPO Regional Function(
    Material=Get material_instance()
    Overlay Material=Get overlay_mid()
    Skeletal Mesh=Get skeletal_mesh()
    Velocity Transform=spine_05
    Velocity Interp Speed=Get Velocity Interp Speed()
    Collision Transform=spine_05
    Invert Collision Transform Axis=true
    Delta Time X=@n1.DeltaTimeX
    Previous Velocity Transform Location=Get Previous Velocity Transform Location_0() -> (Previous Velocity Transform Location_0:Vector)
    Previous Velocity Transform Velocity=Get Previous Velocity Transform Velocity_0() -> (Previous Velocity Transform Velocity_0:Vector)
    Region=2
   )
   > Set WPO Regional Function(
    Material=Get material_instance()
    Overlay Material=Get overlay_mid()
    Skeletal Mesh=Get skeletal_mesh()
    Velocity Transform=spine_05
    Velocity Interp Speed=Get Velocity Interp Speed()
    Collision Transform=clavicle_r
    Invert Collision Transform Axis=true
    Delta Time X=@n1.DeltaTimeX
    Previous Velocity Transform Location=Get Previous Velocity Transform Location_1() -> (Previous Velocity Transform Location_1:Vector)
    Previous Velocity Transform Velocity=Get Previous Velocity Transform Velocity_1() -> (Previous Velocity Transform Velocity_1:Vector)
    Region=3
   )
   n5 = Branch(execute=@n4.then_1, Condition=Get bIsSkyDiving())
   > n6 = Set WPO Pho Wind Intensity(WPO Pho Wind Intensity=1)
   SetScalarParameterValue(execute=[@n7.then, @n6.then], self=Get material_instance(), ParameterName=ClothSim_WindStrength, Value=Get WPO Pho Wind Intensity())
   n7 = Set WPO Pho Wind Intensity(execute=@n5.else, WPO Pho Wind Intensity=0)
   Set spine_twist_pos_delta_ref_alpha(
    execute=@n2.then_1
    spine_twist_pos_delta_ref_alpha:double=GetAlphaFromReferenceTransformDelta(
     SkeletalMeshComponent=Get skeletal_mesh()
     BoneName=spine_05
     BaseName=pelvis
     DeltaComponentType=QuaternionTwist
     InputRange=(X=3,Y=12)
     TwistAxis="0,1,0"
    )
   ) -> (Output_Get:double)
   Event BlueprintInitializeAnimation()
   > Set skeletal_mesh(skeletal_mesh:Skeletal Mesh Component Object Reference=GetPartSkeletalMeshComponent()) -> (Output_Get:Skeletal Mesh Component Object Reference)
   Event OnCharacterCustomizationLoaded()
   > n8 = Sequence()
   n9 = TryGetPawnOwner()
   > Is Valid(InputObject=@n9)
   > n10 = Cast To FortPlayerPawnAthena(Object=@n9)
   > Set Player Pawn(Player Pawn:Fort Player Pawn Athena Object Reference=@n10) -> (Output_Get:Fort Player Pawn Athena Object Reference)
   n11 = Get skeletal_mesh()
   n12 = Get Dynamic Material Instance(execute=@n8.then_1, SkeletalMeshComponent=@n11)
   > Set material_instance(material_instance:Material Instance Dynamic Object Reference=@n12) -> (Output_Get:Material Instance Dynamic Object Reference)
   > n13 = GetOrCreateDynamicOverlayMaterial(MeshComponent=@n11, MaterialName=None)
   > Set overlay_mid(overlay_mid:Material Instance Dynamic Object Reference=@n13) -> (Output_Get:Material Instance Dynamic Object Reference)
  ]}
]
variables [
 fur_alpha {type=number, default=0.5}
 gravity_override {type=Vector}
 Local Noise Offset {type=Vector}
 "Local Noise Transform  Location" {type=Vector}
 material_instance {type=MaterialInstanceDynamic}
 overlay_mid {type=MaterialInstanceDynamic}
 Player Pawn {type=FortPlayerPawnAthena}
 Previous Velocity Transform Location {type=Vector}
 Previous Velocity Transform Location_0 {type=Vector}
 Previous Velocity Transform Location_1 {type=Vector}
 Previous Velocity Transform Velocity {type=Vector}
 Previous Velocity Transform Velocity_0 {type=Vector}
 Previous Velocity Transform Velocity_1 {type=Vector}
 skeletal_mesh {type=SkeletalMeshComponent}
 spine_twist_pos_delta_ref_alpha {type=number}
 Velocity Interp Speed {type=number, default=30}
 WPO Pho Wind Intensity {type=number}
]
