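# Physics asset for the male medium InfernoCoat player body, named as the collider set for its cape.
# Capsules cover the pelvis, spine, clavicles and neck, with a sphere on the head; arm, leg and foot bodies have no shapes.
# Capsules also follow the dyn_skirt bone chains rooted at pelvis and spine_01, with front-side chains only on the right.
# All constraints leave swing and twist free, so the 45-degree limits are inactive.
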PhysicsAsset /BRCosmetics/Characters/Player/Male/Medium/Bodies/M_MED_InfernoCoat/Meshes/M_MED_InfernoCoat_CapeColliders
bodies {
 calf_l []
 calf_r []
 clavicle_l [{shapeType=Capsule, length=7.7636, radius=11.1427, center=<9.9642 -0.7741 -1.5923>, rotation=<rot 75.9352 -0.0001 175.5622>}]
 clavicle_r [{shapeType=Capsule, length=7.7636, radius=11.1427, center=<-9.9642 0.7741 1.5923>, rotation=<rot 75.9351 -0.0001 -4.4378>}]
 dyn_skirt_back_01_l [{
   shapeType=Capsule
   length=9.3561
   radius=6.4639
   shapeName=dyn_skirt_back_01_l_capsule
   center=<6.2208 -0.0827 0.5725>
   rotation=<rot 0 90.2554 -57.7586>
  }]
 dyn_skirt_back_01_r [{
   shapeType=Capsule
   length=11.6548
   radius=6.4639
   shapeName=dyn_skirt_back_01_r_capsule
   center=<4.9703 0.2567 0.2933>
   rotation=<rot 0 -87.874 -68.3006>
  }]
 dyn_skirt_back_02 [{
   shapeType=Capsule
   length=17.3853
   radius=7.9845
   shapeName=dyn_skirt_back_02_capsule
   center=<5.521 0.9765 1.6943>
   rotation=<rot 0 -93.7501 179.1769>
  }]
 dyn_skirt_back_02_l [{
   shapeType=Capsule
   length=9.4026
   radius=6.4639
   shapeName=dyn_skirt_back_02_l_capsule
   center=<7.2098 0.5926 -0.7013>
   rotation=<rot 0 -85.363 -79.6666>
  }]
 dyn_skirt_back_02_r [{
   shapeType=Capsule
   length=8.0907
   radius=6.4639
   shapeName=dyn_skirt_back_02_r_capsule
   center=<5.5495 0.5666 1.0963>
   rotation=<rot 0 93.3619 -83.3823>
  }]
 dyn_skirt_back_03 [{shapeType=Capsule, length=0.505, radius=0.505, shapeName=dyn_skirt_back_03_capsule, rotation=<rot 0 -90 0>}]
 dyn_skirt_back_side_02_l [{shapeType=Capsule, length=0.505, radius=0.505, shapeName=dyn_skirt_back_side_02_l_capsule, rotation=<rot 0 -90 0>}]
 dyn_skirt_back_side_02_r [{shapeType=Capsule, length=0.505, radius=0.505, shapeName=dyn_skirt_back_side_02_r_capsule, rotation=<rot 0 -90 0>}]
 dyn_skirt_back_side_03_l [{
   shapeType=Capsule
   length=8.6466
   radius=7.5991
   shapeName=dyn_skirt_back_side_03_l_capsule
   center=<2.8241 -0.0901 4.9829>
   rotation=<rot 4.5347 -90 -1.943>
  }]
 dyn_skirt_back_side_03_r [{
   shapeType=Capsule
   length=5.7585
   radius=5.5807
   shapeName=dyn_skirt_back_side_03_r_capsule
   center=<3.0318 0.0557 -0.4917>
   rotation=<rot 0 -101.8564 -47.732>
  }]
 dyn_skirt_back_side_04_l [{
   shapeType=Capsule
   length=10.389
   radius=9.1281
   shapeName=dyn_skirt_back_side_04_l_capsule
   center=<6.0452 -0.4384 4.8363>
   rotation=<rot 0 -91.5684 -88.0886>
  }]
 dyn_skirt_back_side_04_r [{
   shapeType=Capsule
   length=12.4036
   radius=6.5383
   shapeName=dyn_skirt_back_side_04_r_capsule
   center=<7.6956 -0.2307 -1.8728>
   rotation=<rot 0 -91.9201 -86.1765>
  }]
 dyn_skirt_front_side_03_r [{
   shapeType=Capsule
   length=6.9253
   radius=4.0796
   shapeName=dyn_skirt_front_side_03_r_capsule
   center=<5.5693 0.2202 -0.8237>
   rotation=<rot 0 -85.2108 -70.6367>
  }]
 dyn_skirt_front_side_04_r [{
   shapeType=Capsule
   length=12.9408
   radius=9.2963
   shapeName=dyn_skirt_front_side_04_r_capsule
   center=<9.045 1.8622 -1.9538>
   rotation=<rot 0 84.7391 -46.2383>
  }]
 dyn_skirt_front_side_05_r [{
   shapeType=Capsule
   length=14.9589
   radius=8.0788
   shapeName=dyn_skirt_front_side_05_r_capsule
   center=<4.5616 2.379 0.4031>
   rotation=<rot 4.4542 -90 -27.6752>
  }]
 dyn_skirt_side_02_l [{
   shapeType=Capsule
   length=8.0871
   radius=4.5547
   shapeName=dyn_skirt_side_02_l_capsule
   center=<2.6741 0.6112 -0.2696>
   rotation=<rot 0 83.2835 -66.9383>
  }]
 dyn_skirt_side_02_r [{
   shapeType=Capsule
   length=11.6875
   radius=7.2692
   shapeName=dyn_skirt_side_02_r_capsule
   center=<3.1314 0.8228 0.2022>
   rotation=<rot 0 -96.3567 -66.1666>
  }]
 dyn_skirt_side_03_l [{shapeType=Capsule, length=0.505, radius=0.505, shapeName=dyn_skirt_side_03_l_capsule, rotation=<rot 0 -90 0>}]
 dyn_skirt_side_03_r [{
   shapeType=Capsule
   length=2.6688
   radius=0.5274
   shapeName=dyn_skirt_side_03_r_capsule
   center=<0.8892 -0.1395 0.3455>
   rotation=<rot 11.85 -90 -4.016>
  }]
 dyn_skirt_side_04_l [{shapeType=Capsule, length=0.505, radius=0.505, shapeName=dyn_skirt_side_04_l_capsule, rotation=<rot 0 -90 0>}]
 dyn_skirt_side_04_r [{
   shapeType=Capsule
   length=13.0955
   radius=8.5369
   shapeName=dyn_skirt_side_04_r_capsule
   center=<4.7126 0.3502 -0.8329>
   rotation=<rot 0 -92.0697 -76.6743>
  }]
 dyn_skirt_side_05_l [{
   shapeType=Capsule
   length=10.8868
   radius=8.2537
   shapeName=dyn_skirt_side_05_l_capsule
   center=<7.7872 1.4393 0.1475>
   rotation=<rot 0 86.9353 -68.8628>
  }]
 foot_l []
 foot_r []
 head [{shapeType=Sphere, radius=17.3125, center=<11.028 -0.174 0>}]
 lowerarm_l []
 lowerarm_r []
 neck_02 [{shapeType=Capsule, length=13.4375, radius=7.3175, center=<-0.2642 -0.3408 0>, rotation=<rot 90 -171.1193 -175.3978>}]
 pelvis [
  {shapeType=Capsule, length=8.6562, radius=13.5388, center=<-2.6215 -1.3695 0>}
  {shapeType=Capsule, length=20.5104, radius=16.9357, shapeName=pelvis_capsule, center=<-21.6915 -3.8336 0>}
  {shapeType=Capsule, length=22.5121, radius=20.3824, shapeName=pelvis_capsule, center=<-42.5702 -7.3979 0>}
 ]
 spine_01 [{shapeType=Capsule, length=10.7187, radius=12.4138, center=<-2.6215 -0.7156 0>}]
 spine_02 [
  {shapeType=Capsule, length=10.5937, radius=10.945, center=<-2.6215 0.4993 0.0001>}
  {shapeType=Capsule, length=29.875, radius=11.1013, center=<2.5024 0.4992 3.9825>, rotation=<rot 90 0.0049 0.0049>}
  {shapeType=Capsule, length=29.875, radius=11.1013, center=<2.4704 0.4991 -3.9825>, rotation=<rot 90 -0.0396 -0.0396>}
 ]
 spine_03 [{shapeType=Capsule, length=9.1562, radius=10.7575, center=<-1.953 0.1364 0.0001>}]
 spine_04 [{shapeType=Capsule, length=14.6562, radius=13.6325, center=<2.1467 -1.2748 0.0002>}]
 spine_05 [
  {shapeType=Capsule, length=17.8437, radius=17.195, center=<-1.7514 2.4879 0>}
  {shapeType=Capsule, length=12.7812, radius=17.0625, center=<0.7875 3.1143 0>}
 ]
 thigh_l []
 thigh_r []
 upperarm_l []
 upperarm_r []
}
constraints [
 {
  bone1Name=spine_01
  bone2Name=pelvis
  swing1LimitDegrees=45
  swing1Motion=Free
  swing2LimitDegrees=45
  swing2Motion=Free
  twistLimitDegrees=45
  twistMotion=Free
 }
 {
  bone1Name=spine_02
  bone2Name=spine_01
  swing1LimitDegrees=45
  swing1Motion=Free
  swing2LimitDegrees=45
  swing2Motion=Free
  twistLimitDegrees=45
  twistMotion=Free
 }
 {
  bone1Name=spine_03
  bone2Name=spine_02
  swing1LimitDegrees=45
  swing1Motion=Free
  swing2LimitDegrees=45
  swing2Motion=Free
  twistLimitDegrees=45
  twistMotion=Free
 }
 {
  bone1Name=spine_04
  bone2Name=spine_03
  swing1LimitDegrees=45
  swing1Motion=Free
  swing2LimitDegrees=45
  swing2Motion=Free
  twistLimitDegrees=45
  twistMotion=Free
 }
 {
  bone1Name=spine_05
  bone2Name=spine_04
  swing1LimitDegrees=45
  swing1Motion=Free
  swing2LimitDegrees=45
  swing2Motion=Free
  twistLimitDegrees=45
  twistMotion=Free
 }
 {
  bone1Name=clavicle_l
  bone2Name=spine_05
  swing1LimitDegrees=45
  swing1Motion=Free
  swing2LimitDegrees=45
  swing2Motion=Free
  twistLimitDegrees=45
  twistMotion=Free
 }
 {
  bone1Name=upperarm_l
  bone2Name=clavicle_l
  swing1LimitDegrees=45
  swing1Motion=Free
  swing2LimitDegrees=45
  swing2Motion=Free
  twistLimitDegrees=45
  twistMotion=Free
 }
 {
  bone1Name=lowerarm_l
  bone2Name=upperarm_l
  swing1LimitDegrees=45
  swing1Motion=Free
  swing2LimitDegrees=45
  swing2Motion=Free
  twistLimitDegrees=45
  twistMotion=Free
 }
 {
  bone1Name=neck_02
  bone2Name=spine_05
  swing1LimitDegrees=45
  swing1Motion=Free
  swing2LimitDegrees=45
  swing2Motion=Free
  twistLimitDegrees=45
  twistMotion=Free
 }
 {
  bone1Name=head
  bone2Name=neck_02
  swing1LimitDegrees=45
  swing1Motion=Free
  swing2LimitDegrees=45
  swing2Motion=Free
  twistLimitDegrees=45
  twistMotion=Free
 }
 {
  bone1Name=thigh_l
  bone2Name=pelvis
  swing1LimitDegrees=45
  swing1Motion=Free
  swing2LimitDegrees=45
  swing2Motion=Free
  twistLimitDegrees=45
  twistMotion=Free
 }
 {
  bone1Name=calf_l
  bone2Name=thigh_l
  swing1LimitDegrees=45
  swing1Motion=Free
  swing2LimitDegrees=45
  swing2Motion=Free
  twistLimitDegrees=45
  twistMotion=Free
 }
 {
  bone1Name=foot_l
  bone2Name=calf_l
  swing1LimitDegrees=45
  swing1Motion=Free
  swing2LimitDegrees=45
  swing2Motion=Free
  twistLimitDegrees=45
  twistMotion=Free
 }
 {
  bone1Name=clavicle_r
  bone2Name=spine_05
  swing1LimitDegrees=45
  swing1Motion=Free
  swing2LimitDegrees=45
  swing2Motion=Free
  twistLimitDegrees=45
  twistMotion=Free
 }
 {
  bone1Name=upperarm_r
  bone2Name=clavicle_r
  swing1LimitDegrees=45
  swing1Motion=Free
  swing2LimitDegrees=45
  swing2Motion=Free
  twistLimitDegrees=45
  twistMotion=Free
 }
 {
  bone1Name=lowerarm_r
  bone2Name=upperarm_r
  swing1LimitDegrees=45
  swing1Motion=Free
  swing2LimitDegrees=45
  swing2Motion=Free
  twistLimitDegrees=45
  twistMotion=Free
 }
 {
  bone1Name=thigh_r
  bone2Name=pelvis
  swing1LimitDegrees=45
  swing1Motion=Free
  swing2LimitDegrees=45
  swing2Motion=Free
  twistLimitDegrees=45
  twistMotion=Free
 }
 {
  bone1Name=calf_r
  bone2Name=thigh_r
  swing1LimitDegrees=45
  swing1Motion=Free
  swing2LimitDegrees=45
  swing2Motion=Free
  twistLimitDegrees=45
  twistMotion=Free
 }
 {
  bone1Name=foot_r
  bone2Name=calf_r
  swing1LimitDegrees=45
  swing1Motion=Free
  swing2LimitDegrees=45
  swing2Motion=Free
  twistLimitDegrees=45
  twistMotion=Free
 }
 {
  bone1Name=dyn_skirt_side_05_l
  bone2Name=dyn_skirt_side_04_l
  swing1LimitDegrees=45
  swing1Motion=Free
  swing2LimitDegrees=45
  swing2Motion=Free
  twistLimitDegrees=45
  twistMotion=Free
 }
 {
  bone1Name=dyn_skirt_side_03_l
  bone2Name=dyn_skirt_side_02_l
  swing1LimitDegrees=45
  swing1Motion=Free
  swing2LimitDegrees=45
  swing2Motion=Free
  twistLimitDegrees=45
  twistMotion=Free
 }
 {
  bone1Name=dyn_skirt_side_04_l
  bone2Name=dyn_skirt_side_03_l
  swing1LimitDegrees=45
  swing1Motion=Free
  swing2LimitDegrees=45
  swing2Motion=Free
  twistLimitDegrees=45
  twistMotion=Free
 }
 {
  bone1Name=dyn_skirt_side_02_l
  bone2Name=pelvis
  swing1LimitDegrees=45
  swing1Motion=Free
  swing2LimitDegrees=45
  swing2Motion=Free
  twistLimitDegrees=45
  twistMotion=Free
 }
 {
  bone1Name=dyn_skirt_side_02_r
  bone2Name=pelvis
  swing1LimitDegrees=45
  swing1Motion=Free
  swing2LimitDegrees=45
  swing2Motion=Free
  twistLimitDegrees=45
  twistMotion=Free
 }
 {
  bone1Name=dyn_skirt_side_03_r
  bone2Name=dyn_skirt_side_02_r
  swing1LimitDegrees=45
  swing1Motion=Free
  swing2LimitDegrees=45
  swing2Motion=Free
  twistLimitDegrees=45
  twistMotion=Free
 }
 {
  bone1Name=dyn_skirt_side_04_r
  bone2Name=dyn_skirt_side_03_r
  swing1LimitDegrees=45
  swing1Motion=Free
  swing2LimitDegrees=45
  swing2Motion=Free
  twistLimitDegrees=45
  twistMotion=Free
 }
 {
  bone1Name=dyn_skirt_back_02_l
  bone2Name=dyn_skirt_back_01_l
  swing1LimitDegrees=45
  swing1Motion=Free
  swing2LimitDegrees=45
  swing2Motion=Free
  twistLimitDegrees=45
  twistMotion=Free
 }
 {
  bone1Name=dyn_skirt_back_01_l
  bone2Name=dyn_skirt_back_03
  swing1LimitDegrees=45
  swing1Motion=Free
  swing2LimitDegrees=45
  swing2Motion=Free
  twistLimitDegrees=45
  twistMotion=Free
 }
 {
  bone1Name=dyn_skirt_back_03
  bone2Name=dyn_skirt_back_02
  swing1LimitDegrees=45
  swing1Motion=Free
  swing2LimitDegrees=45
  swing2Motion=Free
  twistLimitDegrees=45
  twistMotion=Free
 }
 {
  bone1Name=dyn_skirt_back_02
  bone2Name=spine_01
  swing1LimitDegrees=45
  swing1Motion=Free
  swing2LimitDegrees=45
  swing2Motion=Free
  twistLimitDegrees=45
  twistMotion=Free
 }
 {
  bone1Name=dyn_skirt_back_02_r
  bone2Name=dyn_skirt_back_01_r
  swing1LimitDegrees=45
  swing1Motion=Free
  swing2LimitDegrees=45
  swing2Motion=Free
  twistLimitDegrees=45
  twistMotion=Free
 }
 {
  bone1Name=dyn_skirt_back_01_r
  bone2Name=dyn_skirt_back_03
  swing1LimitDegrees=45
  swing1Motion=Free
  swing2LimitDegrees=45
  swing2Motion=Free
  twistLimitDegrees=45
  twistMotion=Free
 }
 {
  bone1Name=dyn_skirt_back_side_04_l
  bone2Name=dyn_skirt_back_side_03_l
  swing1LimitDegrees=45
  swing1Motion=Free
  swing2LimitDegrees=45
  swing2Motion=Free
  twistLimitDegrees=45
  twistMotion=Free
 }
 {
  bone1Name=dyn_skirt_back_side_03_l
  bone2Name=dyn_skirt_back_side_02_l
  swing1LimitDegrees=45
  swing1Motion=Free
  swing2LimitDegrees=45
  swing2Motion=Free
  twistLimitDegrees=45
  twistMotion=Free
 }
 {
  bone1Name=dyn_skirt_back_side_02_l
  bone2Name=spine_01
  swing1LimitDegrees=45
  swing1Motion=Free
  swing2LimitDegrees=45
  swing2Motion=Free
  twistLimitDegrees=45
  twistMotion=Free
 }
 {
  bone1Name=dyn_skirt_back_side_04_r
  bone2Name=dyn_skirt_back_side_03_r
  swing1LimitDegrees=45
  swing1Motion=Free
  swing2LimitDegrees=45
  swing2Motion=Free
  twistLimitDegrees=45
  twistMotion=Free
 }
 {
  bone1Name=dyn_skirt_back_side_03_r
  bone2Name=dyn_skirt_back_side_02_r
  swing1LimitDegrees=45
  swing1Motion=Free
  swing2LimitDegrees=45
  swing2Motion=Free
  twistLimitDegrees=45
  twistMotion=Free
 }
 {
  bone1Name=dyn_skirt_back_side_02_r
  bone2Name=spine_01
  swing1LimitDegrees=45
  swing1Motion=Free
  swing2LimitDegrees=45
  swing2Motion=Free
  twistLimitDegrees=45
  twistMotion=Free
 }
 {
  bone1Name=dyn_skirt_front_side_03_r
  bone2Name=pelvis
  swing1LimitDegrees=45
  swing1Motion=Free
  swing2LimitDegrees=45
  swing2Motion=Free
  twistLimitDegrees=45
  twistMotion=Free
 }
 {
  bone1Name=dyn_skirt_front_side_04_r
  bone2Name=dyn_skirt_front_side_03_r
  swing1LimitDegrees=45
  swing1Motion=Free
  swing2LimitDegrees=45
  swing2Motion=Free
  twistLimitDegrees=45
  twistMotion=Free
 }
 {
  bone1Name=dyn_skirt_front_side_05_r
  bone2Name=dyn_skirt_front_side_04_r
  swing1LimitDegrees=45
  swing1Motion=Free
  swing2LimitDegrees=45
  swing2Motion=Free
  twistLimitDegrees=45
  twistMotion=Free
 }
]
